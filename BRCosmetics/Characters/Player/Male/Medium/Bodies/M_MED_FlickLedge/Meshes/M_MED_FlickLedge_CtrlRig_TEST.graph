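# Test Control Rig for the FlickLedge male medium body that procedurally deforms robe skirt, fold and brooch bones to follow the body.
# A smoothed Leg ROM function supplies leg pose weights that drive mirrored corrective transforms and animation curves.
# Thigh and knee solvers blend skirt bones toward twist-free thigh helpers, calves and feet, pushing thigh-constrained joints outward by their displacement.
# A knee-sliding branch sets the Leg ROM alpha, whose inverse weights alternate skirt offsets.
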ControlRigBlueprint /BRCosmetics/Characters/Player/Male/Medium/Bodies/M_MED_FlickLedge/Meshes/M_MED_FlickLedge_CtrlRig_TEST
graphs [
 Maximum_1_ContainedGraph {nodes [
   n1 = RigVMFunctionEntryNode()
   Maximum = &RigVMFunction_MathDoubleMax(A=@n1.A, B=@n1.B)
   RigVMFunctionReturnNode(Result=&RigVMFunction_MathDoubleMax(B=@n1.C, A=@Maximum.Result).Result)
  ]}
 Maximum_1_ContainedGraph {nodes [
   n1 = RigVMFunctionEntryNode()
   Maximum = &RigVMFunction_MathDoubleMax(A=@n1.A, B=@n1.B)
   RigVMFunctionReturnNode(Result=&RigVMFunction_MathDoubleMax(B=@n1.C, A=@Maximum.Result).Result)
  ]}
 Maximum_1_ContainedGraph {nodes [
   n1 = RigVMFunctionEntryNode()
   Maximum = &RigVMFunction_MathDoubleMax(A=@n1.A, B=@n1.B)
   RigVMFunctionReturnNode(Result=&RigVMFunction_MathDoubleMax(B=@n1.D, A=&RigVMFunction_MathDoubleMax(B=@n1.C, A=@Maximum.Result).Result).Result)
  ]}
 New Function_1_ContainedGraph {nodes [
   n1 = RigVMFunctionEntryNode()
   > RigVMFunction_Sequence = RigVMAggregateNode()
   RigVMFunctionReturnNode(ExecuteContext=@RigVMFunction_Sequence.C)
   n2 = &RigVMFunction_MathDoubleRemap(SourceMaximum=0.3, SourceMinimum=0, TargetMaximum=1, TargetMinimum=0, Value=@n1.Thigh Side Spr)
   Clamp = &RigVMFunction_MathDoubleClamp(Maximum=0.5, Value=@n2.Result)
   n3 = &RigVMFunction_MathDoubleRemap(SourceMaximum=0.3, SourceMinimum=0, TargetMaximum=1, TargetMinimum=0, Value=@n1.Thigh Fwd Side Spr)
   Remap = &RigVMFunction_MathDoubleRemap(SourceMaximum=0.3, SourceMinimum=0, TargetMaximum=1, TargetMinimum=0, Value=@n1.Thigh Fwd Spr)
   Maximum = RigVMAggregateNode(C=@Clamp.Result, B=@n3.Result, A=@Remap.Result)
   n4 = &RigVMFunction_MathDoubleRemap(SourceMaximum=0.3, SourceMinimum=0, TargetMaximum=1, TargetMinimum=0, Value=@n1.Thigh Rear Spr)
   Subtract = &RigVMFunction_MathDoubleSub(A=1, B=@n4.Result)
   Multiply = &RigVMFunction_MathDoubleMul(A=@Maximum.Result, B=@Subtract.Result)
   &RigUnit_ParentConstraint(
    AdvancedSettings=(InterpolationType=Average,RotationOrderForFilter=XZY)
    Filter=(TranslationFilter=(bX=True,bY=True,bZ=True),RotationFilter=(bX=true,bY=true,bZ=true),ScaleFilter=(bX=True,bY=True,bZ=True))
    Parents=((Item=(Type=Bone,Name="thigh_l"),Weight=1))
    bMaintainOffset=true
    Weight=&RigVMFunction_MathDoubleRemap(SourceMaximum=1, SourceMinimum=0, TargetMaximum=1, TargetMinimum=0.3, Value=@Multiply.Result).Result
    ExecutePin=@RigVMFunction_Sequence.A
    Child=@n1.Bind Front Skirt Joint
    Parents.0.Item=@n1.Thigh Joint
   )
   n5 = &RigVMFunction_MathDoubleRemap(SourceMaximum=0.3, SourceMinimum=0, TargetMaximum=1, TargetMinimum=0, Value=@n1.Thigh Side Rear Spr)
   Maximum_1 = RigVMAggregateNode(
    A=&RigVMFunction_MathDoubleClamp(Maximum=0.3, Value=@n3.Result).Result
    C=&RigVMFunction_MathDoubleClamp(Maximum=0.3, Value=@n5.Result).Result
    B=&RigVMFunction_MathDoubleRemap(SourceMaximum=1, SourceMinimum=0, TargetMaximum=1, TargetMinimum=0, Value=@n1.Thigh Side Spr).Result
   )
   n6 = &RigVMFunction_MathDoubleSub(A=1, B=@Maximum_1.Result)
   > &RigUnit_ParentConstraint(
    AdvancedSettings=(InterpolationType=Average,RotationOrderForFilter=XZY)
    Filter=(TranslationFilter=(bX=True,bY=True,bZ=True),RotationFilter=(bX=true,bY=true,bZ=true),ScaleFilter=(bX=True,bY=True,bZ=True))
    Parents=((Item=(Type=Bone,Name="thigh_l"),Weight=1),(Item=(Type=Bone,Name="None"),Weight=1),(Item=(Type=Bone,Name="None"),Weight=1))
    Weight=1
    bMaintainOffset=true
    Parents.2.Weight=@Maximum_1.Result
    Parents.0.Weight=@n6.Result
    Parents.1.Weight=@n6.Result
    Parents.1.Item=@n1.Bind Front Skirt Joint
    Child=@n1.Bind Side Skirt Joint
    Parents.0.Item=@n1.Bind Rear Skirt Joint
    Parents.2.Item=@n1.Thigh Joint
   )
   Maximum_2 = RigVMAggregateNode(
    A=&RigVMFunction_MathDoubleClamp(Maximum=0.5, Value=@n2.Result).Result
    D=&RigVMFunction_MathDoubleClamp(Maximum=0.5, Value=@Remap.Result).Result
    C=@n4.Result
    B=@n5.Result
   )
   n7 = &RigVMFunction_MathDoubleSub(A=1, B=@Remap.Result)
   > &RigUnit_ParentConstraint(
    AdvancedSettings=(InterpolationType=Average,RotationOrderForFilter=XZY)
    Parents=((Item=(Type=Bone,Name="thigh_l"),Weight=1),(Item=(Type=Bone,Name="None"),Weight=0))
    bMaintainOffset=true
    Parents.1.Item=/Script/ControlRig.RigUnit_HierarchyGetParent(bDefaultParent=True, Child=@n1.Bind Rear Skirt Joint) {is=struct RigVMUnitNode}.Parent
    Weight=&RigVMFunction_MathDoubleRemap(
     SourceMaximum=1
     SourceMinimum=0
     TargetMaximum=1
     TargetMinimum=0.3
     Value=&RigVMFunction_MathDoubleMul(A=@Maximum_2.Result, B=@n7.Result).Result
    ).Result
    Child=@n1.Bind Rear Skirt Joint
    Parents.0.Item=@n1.Thigh Joint
   )
   RigVMFunction_MathVectorDistance = &RigVMFunction_MathVectorDistance(
    A=&RigUnit_GetTransform(Space=LocalSpace, bInitial=False, Item=@n1.Bind Front Skirt Joint).Transform.Translation
    B=&RigUnit_GetTransform(Space=LocalSpace, bInitial=true, Item=@n1.Bind Front Skirt Joint).Transform.Translation
   )
   > &RigUnit_OffsetTransformForItem(
    OffsetTransform=(Rotation=(X=0,Y=0,Z=0,W=1),Translation=(X=0,Y=0,Z=0),Scale3D=(X=1,Y=1,Z=1))
    Weight=1
    bPropagateToChildren=True
    OffsetTransform.Translation=/Script/RigVM.RigVMFunction_MathVectorMul(A=(X=1,Y=0,Z=0), A.X=@RigVMFunction_MathVectorDistance.Result, B=@n1.Push Vector) {is=struct RigVMUnitNode, template="Multiply::Execute(in A,in B,out Result)"}.Result
    Item=@n1.Bind Front Skirt Joint
   )
   n8 = &RigUnit_GetTransform(Space=LocalSpace, bInitial=False, Item=@n1.Bind Side Skirt Joint)
   n9 = &RigUnit_GetTransform(Space=LocalSpace, bInitial=true, Item=@n1.Bind Side Skirt Joint)
   > &RigUnit_OffsetTransformForItem(
    OffsetTransform=(Rotation=(X=0,Y=0,Z=0,W=1),Translation=(X=0,Y=0,Z=0),Scale3D=(X=1,Y=1,Z=1))
    Weight=1
    bPropagateToChildren=True
    OffsetTransform.Translation=&RigVMFunction_MathVectorMul(
     A=(X=1,Y=0,Z=0)
     A.X=&RigVMFunction_MathVectorDistance(A=@n8.Transform.Translation, B=@n9.Transform.Translation).Result
     B=@n1.Push Vector
    ) {template="Multiply::Execute(in A,in B,out Result)"}.Result
    Item=@n1.Bind Side Skirt Joint
   )
   n10 = &RigUnit_GetTransform(Space=LocalSpace, bInitial=False, Item=@n1.Bind Rear Skirt Joint)
   n11 = &RigUnit_GetTransform(Space=LocalSpace, bInitial=true, Item=@n1.Bind Rear Skirt Joint)
   > &RigUnit_OffsetTransformForItem(
    OffsetTransform=(Rotation=(X=0,Y=0,Z=0,W=1),Translation=(X=0,Y=0,Z=0),Scale3D=(X=1,Y=1,Z=1))
    Weight=1
    bPropagateToChildren=True
    OffsetTransform.Translation=&RigVMFunction_MathVectorMul(
     A=(X=1,Y=0,Z=0)
     A.X=&RigVMFunction_MathVectorDistance(A=@n10.Transform.Translation, B=@n11.Transform.Translation).Result
     B=@n1.Push Vector
    ) {template="Multiply::Execute(in A,in B,out Result)"}.Result
    Item=@n1.Bind Rear Skirt Joint
   )
  ]}
 New Function_1_ContainedGraph {nodes [
   RigVMFunctionEntryNode()
   Remap = &RigVMFunction_MathDoubleRemap(SourceMaximum=0.3, SourceMinimum=0, TargetMaximum=1, TargetMinimum=0, Value=RigVMVariableNode() {variable=Rear Knee Bend SPR}.Value)
   n1 = RigVMVariableNode() {variable=Bind Rear Knee}
   > &RigUnit_AimConstraintLocalSpaceOffset(
    Filter=(bX=True,bY=True,bZ=True)
    Parents=((Item=(Type=Bone,Name="ball_l"),Weight=1))
    WorldUp=(Target=(X=0,Y=0,Z=1),Kind=Location,Space=(Type=Bone,Name="ball_l"))
    Weight=@Remap.Result
    Parents.0.Item=RigVMVariableNode() {variable=Ball}.Value
    Child=@n1.Value
    WorldUp.Space=@n1.Value
    AimAxis=RigVMVariableNode() {variable=AimAxis}.Value
   )
   > &RigUnit_ParentConstraint(
    AdvancedSettings=(InterpolationType=Average,RotationOrderForFilter=XZY)
    Filter=(TranslationFilter=(bX=true,bY=true,bZ=true),RotationFilter=(bX=true,bY=true,bZ=true),ScaleFilter=(bX=True,bY=True,bZ=True))
    Parents=((Item=(Type=Bone,Name="None"),Weight=1))
    bMaintainOffset=true
    Parents.0.Item=RigVMVariableNode() {variable=Knee}.Value
    Child=RigVMVariableNode() {variable=Bind Front Knee}.Value
    Weight=RigVMVariableNode() {variable=Front Knee Bend SPR}.Value
   )
   > RigVMFunctionReturnNode()
  ]}
 New Function_ContainedGraph {nodes [
   RigVMFunctionEntryNode()
   > RigVMFunction_Sequence = RigVMAggregateNode()
   KalmanFilter = &RigVMFunction_KalmanFloat(BufferSize=4, Value=RigVMVariableNode() {variable=Leg ROM_Result Alpha}.Value) {template="KalmanFilter::Execute(in Value,in BufferSize,out Result)"}
   Leg ROM = RigVMFunctionReferenceNode(
    Include Ankles=true
    Include Distance Checks=true
    Include Knees=true
    Interp Output Speed=10
    Interp Results=true
    SPR Optional Parent Item=(Type=Bone,Name="pelvis")
    Result Alpha=@KalmanFilter.Result
    ExecuteContext=@RigVMFunction_Sequence.B
   )
   > RigVMVariableNode(Value=@"Leg ROM".L Thigh Raised Fwd) {variable=Leg ROM_L Thigh Raised Fwd}
   > RigVMVariableNode(Value=@"Leg ROM".R Thigh Raised Fwd) {variable=Leg ROM_R Thigh Raised Fwd}
   > RigVMVariableNode(Value=@"Leg ROM".L Thigh Raised Fwd Side Inbetween) {variable=Leg ROM_L Thigh Raised Fwd Side Inbetween}
   > RigVMVariableNode(Value=@"Leg ROM".R Thigh Raised Fwd Side Inbetween) {variable=Leg ROM_R Thigh Raised Fwd Side Inbetween}
   > RigVMVariableNode(Value=@"Leg ROM".L Thigh Raised Side) {variable=Leg ROM_L Thigh Raised Side}
   > RigVMVariableNode(Value=@"Leg ROM".R Thigh Raised Side) {variable=Leg ROM_R Thigh Raised Side}
   > RigVMVariableNode(Value=@"Leg ROM".L Thigh Raised Side Rear Inbetween) {variable=Leg ROM_L Thigh Raised Side Rear Inbetween}
   > RigVMVariableNode(Value=@"Leg ROM".R Thigh Raised Side Rear Inbetween) {variable=Leg ROM_R Thigh Raised Side Rear Inbetween}
   > RigVMVariableNode(Value=@"Leg ROM".L Thigh Raised Rear) {variable=Leg ROM_L Thigh Raised Rear}
   > RigVMVariableNode(Value=@"Leg ROM".R Thigh Raised Rear) {variable=Leg ROM_R Thigh Raised Rear}
   > RigVMVariableNode(Value=@"Leg ROM".L Thigh Across R) {variable=Leg ROM_L Thigh Across R}
   > RigVMVariableNode(Value=@"Leg ROM".R Thigh Across L) {variable=Leg ROM_R Thigh Across L}
   > RigVMVariableNode(Value=@"Leg ROM".L Leg Across R Leg Side) {variable=Leg ROM_L Leg Across R Leg Side}
   > RigVMVariableNode(Value=@"Leg ROM".R Leg Across L Leg Side) {variable=Leg ROM_R Leg Across L Leg Side}
   > RigVMVariableNode(Value=@"Leg ROM".L Thigh Fwd Knee Bend) {variable=Leg ROM_L Thigh Fwd Knee Bend}
   > RigVMVariableNode(Value=@"Leg ROM".R Thigh Fwd Knee Bend) {variable=Leg ROM_R Thigh Fwd Knee Bend}
   > RigVMVariableNode(Value=@"Leg ROM".L Thigh Rear Knee Bend) {variable=Leg ROM_L Thigh Rear Knee Bend}
   > RigVMVariableNode(Value=@"Leg ROM".R Thigh Rear Knee Bend) {variable=Leg ROM_R Thigh Rear Knee Bend}
   > RigVMVariableNode(Value=@"Leg ROM".L Foot Near Pelvis) {variable=Leg ROM_L Foot Near Pelvis}
   > RigVMVariableNode(Value=@"Leg ROM".R Foot Near Pelvis) {variable=Leg ROM_R Foot Near Pelvis}
   > RigVMVariableNode(Value=@"Leg ROM".L Foot on Thigh R) {variable=Leg ROM_L Foot on Thigh R}
   Remap = &RigVMFunction_MathDoubleRemap(SourceMaximum=0.3, SourceMinimum=0, TargetMaximum=1, TargetMinimum=0, Value=RigVMVariableNode() {variable=Leg ROM_R Thigh Raised Side}.Value)
   &RigUnit_ModifyTransforms(
    ItemToModify="((Item=(Type=Bone,Name=\"dyn_skirt_side_r_MSTR\"),Transform=(Rotation=(X=0,Y=0,Z=-0.0436,W=0.999),Translation=(X=0,Y=0,Z=0),Scale3D=(X=1,Y=1,Z=1))),(Item=(Type=Bone,Name=\"dyn_skirt_back_r_MSTR\"),Transform=(Rotation=(X=0,Y=0,Z=-0.0436,W=0.999),Translation=(X= ...(truncated)"
    WeightMinimum=0
    Weight=@Remap.Result
    ExecutePin=@RigVMFunction_Sequence.D
   )
   > &RigUnit_ModifyTransforms(
    ItemToModify="((Item=(Type=Bone,Name=\"dyn_skirt_side_l_MSTR\"),Transform=(Rotation=(X=0,Y=0,Z=-0.0436,W=0.999),Translation=(X=0,Y=0,Z=0),Scale3D=(X=1,Y=1,Z=1))),(Item=(Type=Bone,Name=\"dyn_skirt_back_l_MSTR\"),Transform=(Rotation=(X=0,Y=0,Z=-0.0436,W=0.999),Translation=(X= ...(truncated)"
    WeightMinimum=0
    Weight=&RigVMFunction_MathDoubleRemap(SourceMaximum=0.3, SourceMinimum=0, TargetMaximum=1, TargetMinimum=0, Value=RigVMVariableNode() {variable=Leg ROM_L Thigh Raised Side}.Value).Result
   )
   RigVMFunction_MathVectorDistance = &RigVMFunction_MathVectorDistance(
    A=&RigUnit_GetTransform(Item=(Type=Bone,Name="pelvis"), bInitial=False).Transform.Translation
    B=&RigUnit_GetTransform(Item=(Type=Bone,Name="root"), bInitial=False).Transform.Translation
   )
   n1 = &RigVMFunction_MathDoubleRemap(SourceMaximum=80, SourceMinimum=45, TargetMaximum=0, TargetMinimum=1, Value=@RigVMFunction_MathVectorDistance.Result)
   n2 = &RigVMFunction_MathDoubleRemap(SourceMaximum=1, SourceMinimum=0, TargetMaximum=1, TargetMinimum=0, Value=@n1.Result)
   > Modify Transforms with Mirror_3 = RigVMFunctionReferenceNode(
    Base Modify Data="((Item=(Type=Bone,Name=\"dyn_skirt_back_l_base\"),Transform=(Rotation=(X=-0.0226,Y=-0.0842,Z=-0.2578,W=0.9623),Translation=(X=0,Y=0,Z=0),Scale3D=(X=1,Y=1,Z=1))),(Item=(Type=Bone,Name=\"dyn_skirt_side_l_MSTR\"),Transform=(Rotation=(X=0,Y=0,Z=0.2588,W=0.9659),Tr ...(truncated)"
    Mirror Modify Data="((Item=(Type=Bone,Name=\"dyn_skirt_back_r_base\"),Transform=(Rotation=(X=0,Y=0,Z=0,W=1),Translation=(X=0,Y=0,Z=0),Scale3D=(X=1,Y=1,Z=1))),(Item=(Type=Bone,Name=\"dyn_skirt_side_r_01\"),Transform=(Rotation=(X=0,Y=0,Z=0,W=1),Translation=(X=0,Y=0,Z=0),Scale3D=(X= ...(truncated)"
    "Mirror Weight "=@n2.Result
    Weight=@n2.Result
   )
   > Get Normalized Distance Between = RigVMFunctionReferenceNode(Debug=true, Max Distance=50, Min Distance=5, Source=(Type=Control,Name="foot_l_ctrl_2"), Target=(Type=Bone,Name="foot_l"))
   > /Script/ControlRig.RigUnit_SetCurveValue(Curve=CR Rear Knee Bend L, Value=@"Get Normalized Distance Between".Result) {is=struct RigVMUnitNode}
   > Get Normalized Distance Between_1 = RigVMFunctionReferenceNode(Debug=true, Max Distance=50, Min Distance=5, Source=(Type=Control,Name="foot_r_ctrl_2"), Target=(Type=Bone,Name="foot_r"))
   > &RigUnit_SetCurveValue(Curve=CR Rear Knee Bend R, Value=@"Get Normalized Distance Between_1".Result)
   > Modify Transforms with Mirror_11 = RigVMFunctionReferenceNode(
    Base Modify Data=((Item=(Type=Bone,Name="dyn_skirt_back_l_base"),Transform=(Rotation=(X=0,Y=0,Z=-0.342,W=0.9397),Translation=(X=0,Y=-5,Z=0),Scale3D=(X=1,Y=1,Z=1))))
    Mirror Modify Data=((Item=(Type=Bone,Name="dyn_skirt_back_r_base"),Transform=(Rotation=(X=0,Y=0,Z=0,W=1),Translation=(X=0,Y=0,Z=0),Scale3D=(X=1,Y=1,Z=1))))
    Weight=@"Get Normalized Distance Between".Result
    "Mirror Weight "=@"Get Normalized Distance Between_1".Result
   )
   RigVMFunctionReturnNode(ExecuteContext=@RigVMFunction_Sequence.H)
   RigUnit_GetTransform = &RigUnit_GetTransform(Item=(Type=Bone,Name="thigh_l"), Space=LocalSpace, bInitial=False)
   RigVMFunction_MathQuaternionSwingTwist = &RigVMFunction_MathQuaternionSwingTwist(TwistAxis=(X=1,Y=0,Z=0), Input=@RigUnit_GetTransform.Transform.Rotation)
   Inverse = /Script/RigVM.RigVMFunction_MathQuaternionInverse(Value=@RigVMFunction_MathQuaternionSwingTwist.Twist) {is=struct RigVMUnitNode, template="Inverse::Execute(in Value,out Result)"}
   RigUnit_SetRotation = &RigUnit_SetRotation(Item=(Type=Bone,Name="cr_thigh_no_twist_l"), Weight=1, bInitial=False, Value=@Inverse.Result, ExecutePin=@RigVMFunction_Sequence.A)
   > &RigUnit_ParentConstraint(
    AdvancedSettings=(InterpolationType=Average,RotationOrderForFilter=XZY)
    Child=(Type=Bone,Name="cr_thigh_r")
    Parents=((Item=(Type=Bone,Name="thigh_r"),Weight=1))
    Weight=1
   )
   n3 = &RigUnit_GetTransform(Item=(Type=Bone,Name="cr_thigh_r"), Space=LocalSpace, bInitial=False)
   > &RigUnit_SetRotation(
    Item=(Type=Bone,Name="cr_thigh_no_twist_r")
    Weight=1
    bInitial=False
    bPropagateToChildren=true
    Value=&RigVMFunction_MathQuaternionInverse(Value=&RigVMFunction_MathQuaternionSwingTwist(TwistAxis=(X=-1,Y=0,Z=0), Input=@n3.Transform.Rotation).Twist) {template="Inverse::Execute(in Value,out Result)"}.Result
   )
   > &RigUnit_ParentConstraint(
    AdvancedSettings=(RotationOrderForFilter=XZY)
    Child=(Type=Bone,Name="deform_brooch")
    Parents="((Item=(Type=Bone,Name=\"spine_04\"),Weight=0.0448),(Item=(Type=Bone,Name=\"deform_pec_r\"),Weight=0.0689),(Item=(Type=Bone,Name=\"upperarm_twist_02_r\"),Weight=0.0727),(Item=(Type=Bone,Name=\"spine_05\"),Weight=0.0762),(Item=(Type=Bone,Name=\"deform_upperarm_up_r\" ...(truncated)"
    Weight=1
   )
   > n4 = &RigUnit_SphericalPoseReader(
    ActiveRegionScaleFactors=(PositiveWidth=1,NegativeWidth=0.5,PositiveHeight=1.25,NegativeHeight=1.25)
    ActiveRegionSize=0.3
    Debug=(bDrawDebug=true,bDraw2D=False,bDrawLocalAxes=False,DebugScale=25,DebugSegments=20,DebugThickness=0.25)
    DriverAxis=(X=-1,Y=0,Z=0)
    DriverItem=(Type=Bone,Name="calf_l")
    FalloffRegionScaleFactors=(PositiveWidth=0.75,NegativeWidth=0.75,PositiveHeight=1,NegativeHeight=1)
    FalloffSize=0.5
    FlipHeightScaling=False
    FlipWidthScaling=False
    OptionalParentItem=(Type=Bone,Name="thigh_l")
    RotationOffset=(X=-90,Y=0,Z=10)
   )
   > RigVMVariableNode(Value=@n4.OutputParam) {variable=Knee_Bend_L}
   > n5 = &RigUnit_SphericalPoseReader(
    ActiveRegionScaleFactors=(PositiveWidth=0.5,NegativeWidth=1,PositiveHeight=1.25,NegativeHeight=1.25)
    ActiveRegionSize=0.3
    Debug=(bDrawDebug=false,bDraw2D=False,bDrawLocalAxes=False,DebugScale=25,DebugSegments=20,DebugThickness=0.25)
    DriverAxis=(X=1,Y=0,Z=0)
    DriverItem=(Type=Bone,Name="calf_r")
    FalloffRegionScaleFactors=(PositiveWidth=0.75,NegativeWidth=0.75,PositiveHeight=1,NegativeHeight=1)
    FalloffSize=0.5
    FlipHeightScaling=False
    FlipWidthScaling=False
    OptionalParentItem=(Type=Bone,Name="thigh_r")
    RotationOffset=(X=90,Y=0,Z=10)
   )
   > RigVMVariableNode(Value=@n5.OutputParam) {variable=Knee_Bend_r}
   n6 = RigVMVariableNode() {variable=Is Knee Sliding}
   n7 = &RigVMFunction_ControlFlowBranch(Condition=@n6.Value, ExecuteContext=@RigVMFunction_Sequence.G)
   Subtract = &RigVMFunction_MathDoubleSub(A=1, B=RigVMVariableNode() {variable=Leg ROM_Result Alpha}.Value)
   n8 = &RigVMFunction_KalmanFloat(BufferSize=4, Value=@Subtract.Result) {template="KalmanFilter::Execute(in Value,in BufferSize,out Result)"}
   Modify Transforms with Mirror = RigVMFunctionReferenceNode(
    Base Modify Data="((Item=(Type=Bone,Name=\"dyn_skirt_side_l_MSTR\"),Transform=(Rotation=(X=0,Y=0,Z=-0.5,W=0.866),Translation=(X=0,Y=0,Z=0),Scale3D=(X=1,Y=1,Z=1))),(Item=(Type=Bone,Name=\"dyn_skirt_back_l_MSTR\"),Transform=(Rotation=(X=0,Y=0,Z=-0.6428,W=0.766),Translation=(X=-2, ...(truncated)"
    Mirror Modify Data="((Item=(Type=Bone,Name=\"dyn_skirt_side_r_MSTR\"),Transform=(Rotation=(X=0,Y=0,Z=0,W=1),Translation=(X=0,Y=0,Z=0),Scale3D=(X=1,Y=1,Z=1))),(Item=(Type=Bone,Name=\"dyn_skirt_back_r_MSTR\"),Transform=(Rotation=(X=0,Y=0,Z=0,W=1),Translation=(X=0,Y=0,Z=0),Scale3D=( ...(truncated)"
    "Mirror Weight "=@n8.Result
    Weight=@n8.Result
    ExecuteContext=@n7.Completed
   )
   > &RigUnit_ModifyTransforms(
    ItemToModify=((Item=(Type=Bone,Name="dyn_skirt_MSTR"),Transform=(Rotation=(X=0,Y=0,Z=0,W=1),Translation=(X=6,Y=0,Z=0),Scale3D=(X=1,Y=1,Z=1))))
    WeightMinimum=0
    Weight=@n8.Result
   )
   RigVMVariableNode(Value=1, ExecuteContext=@n7.False) {variable=Leg ROM_Result Alpha}
   RigVMVariableNode(ExecuteContext=@n7.True) {variable=Leg ROM_Result Alpha}
   Modify Transforms with Mirror_6 = RigVMFunctionReferenceNode(
    Base Modify Data="((Item=(Type=Bone,Name=\"dyn_skirt_front_l_MSTR\"),Transform=(Rotation=(X=0,Y=0,Z=0.0436,W=0.999),Translation=(X=-7,Y=-5,Z=0),Scale3D=(X=1,Y=1,Z=1))),(Item=(Type=Bone,Name=\"dyn_skirt_front_mid_MSTR\"),Transform=(Rotation=(X=0,Y=0,Z=-0.1305,W=0.9914),Translati ...(truncated)"
    Mirror Modify Data="((Item=(Type=Bone,Name=\"dyn_skirt_front_r_MSTR\"),Transform=(Rotation=(X=0,Y=0,Z=0,W=1),Translation=(X=0,Y=0,Z=0),Scale3D=(X=1,Y=1,Z=1))),(Item=(Type=Bone,Name=\"dyn_skirt_front_mid_MSTR\"),Transform=(Rotation=(X=0,Y=0,Z=0,W=1),Translation=(X=0,Y=0,Z=0),Scale ...(truncated)"
    Weight=RigVMVariableNode() {variable=Leg ROM_L Thigh Raised Fwd}.Value
    "Mirror Weight "=RigVMVariableNode() {variable=Leg ROM_R Thigh Raised Fwd}.Value
    ExecuteContext=@RigVMFunction_Sequence.E
   )
   > Modify Transforms with Mirror_7 = RigVMFunctionReferenceNode(
    Base Modify Data="((Item=(Type=Bone,Name=\"dyn_skirt_front_l_MSTR\"),Transform=(Rotation=(X=0.0019,Y=0.0436,Z=-0.0436,W=0.9981),Translation=(X=0,Y=4,Z=0),Scale3D=(X=1,Y=1,Z=1))),(Item=(Type=Bone,Name=\"dyn_skirt_side_l_MSTR\"),Transform=(Rotation=(X=0.0019,Y=0.0436,Z=-0.0436,W= ...(truncated)"
    Mirror Modify Data="((Item=(Type=Bone,Name=\"dyn_skirt_front_r_MSTR\"),Transform=(Rotation=(X=0,Y=0,Z=0,W=1),Translation=(X=0,Y=0,Z=0),Scale3D=(X=1,Y=1,Z=1))),(Item=(Type=Bone,Name=\"dyn_skirt_side_r_MSTR\"),Transform=(Rotation=(X=0,Y=0,Z=0,W=1),Translation=(X=0,Y=0,Z=0),Scale3D= ...(truncated)"
    Weight=RigVMVariableNode() {variable=Leg ROM_L Thigh Raised Fwd Side Inbetween}.Value
    "Mirror Weight "=RigVMVariableNode() {variable=Leg ROM_R Thigh Raised Fwd Side Inbetween}.Value
   )
   > Modify Transforms with Mirror_2 = RigVMFunctionReferenceNode(
    Base Modify Data="((Item=(Type=Bone,Name=\"dyn_skirt_front_l_MSTR\"),Transform=(Rotation=(X=0,Y=0,Z=-0.0193,W=0.9998),Translation=(X=0,Y=0,Z=0),Scale3D=(X=1,Y=1,Z=1))),(Item=(Type=Bone,Name=\"dyn_skirt_front_r_MSTR\"),Transform=(Rotation=(X=0.1099,Y=0.2529,Z=-0.0683,W=0.9588),T ...(truncated)"
    Mirror Modify Data="((Item=(Type=Bone,Name=\"dyn_skirt_front_r_MSTR\"),Transform=(Rotation=(X=0,Y=0,Z=0,W=1),Translation=(X=0,Y=0,Z=0),Scale3D=(X=1,Y=1,Z=1))),(Item=(Type=Bone,Name=\"dyn_skirt_front_l_MSTR\"),Transform=(Rotation=(X=0,Y=0,Z=0,W=1),Translation=(X=0,Y=0,Z=0),Scale3D ...(truncated)"
    Weight=&RigVMFunction_MathDoubleRemap(SourceMaximum=0.4, SourceMinimum=0, TargetMaximum=1, TargetMinimum=0, Value=RigVMVariableNode() {variable=Leg ROM_L Thigh Across R}.Value).Result
    "Mirror Weight "=&RigVMFunction_MathDoubleRemap(SourceMaximum=0.4, SourceMinimum=0, TargetMaximum=1, TargetMinimum=0, Value=RigVMVariableNode() {variable=Leg ROM_R Thigh Across L}.Value).Result
   )
   > Modify Transforms with Mirror_4 = RigVMFunctionReferenceNode(
    Base Modify Data="((Item=(Type=Bone,Name=\"dyn_skirt_back_r_MSTR\"),Transform=(Rotation=(X=-0.0151,Y=-0.0858,Z=-0.173,W=0.9811),Translation=(X=0,Y=0,Z=0),Scale3D=(X=1,Y=1,Z=1))),(Item=(Type=Bone,Name=\"dyn_skirt_back_r_base\"),Transform=(Rotation=(X=0,Y=0,Z=-0.0872,W=0.9962),Tr ...(truncated)"
    Mirror Modify Data="((Item=(Type=Bone,Name=\"dyn_skirt_back_l_MSTR\"),Transform=(Rotation=(X=0,Y=0,Z=0,W=1),Translation=(X=0,Y=0,Z=0),Scale3D=(X=1,Y=1,Z=1))),(Item=(Type=Bone,Name=\"dyn_skirt_back_l_base\"),Transform=(Rotation=(X=0,Y=0,Z=0,W=1),Translation=(X=0,Y=0,Z=0),Scale3D=( ...(truncated)"
    Weight=&RigVMFunction_MathDoubleRemap(SourceMaximum=0.8, SourceMinimum=0, TargetMaximum=1, TargetMinimum=0, Value=RigVMVariableNode() {variable=Leg ROM_L Thigh Across R}.Value).Result
    "Mirror Weight "=&RigVMFunction_MathDoubleRemap(SourceMaximum=0.8, SourceMinimum=0, TargetMaximum=1, TargetMinimum=0, Value=RigVMVariableNode() {variable=Leg ROM_R Thigh Across L}.Value).Result
   )
   > Modify Transforms with Mirror_5 = RigVMFunctionReferenceNode(
    Base Modify Data="((Item=(Type=Bone,Name=\"dyn_skirt_front_l_01\"),Transform=(Rotation=(X=0,Y=0,Z=0.5,W=0.866),Translation=(X=0,Y=0,Z=0),Scale3D=(X=1,Y=1,Z=1))),(Item=(Type=Bone,Name=\"dyn_skirt_side_l_01\"),Transform=(Rotation=(X=-0.0868,Y=0.4924,Z=0.1504,W=0.8529),Translation ...(truncated)"
    Mirror Modify Data="((Item=(Type=Bone,Name=\"dyn_skirt_front_r_01\"),Transform=(Rotation=(X=0,Y=0,Z=0,W=1),Translation=(X=0,Y=0,Z=0),Scale3D=(X=1,Y=1,Z=1))),(Item=(Type=Bone,Name=\"dyn_skirt_side_r_01\"),Transform=(Rotation=(X=0,Y=0,Z=0,W=1),Translation=(X=0,Y=0,Z=0),Scale3D=(X=1 ...(truncated)"
    Weight=RigVMVariableNode() {variable=Leg ROM_L Thigh Fwd Knee Bend}.Value
    "Mirror Weight "=RigVMVariableNode() {variable=Leg ROM_R Thigh Fwd Knee Bend}.Value
   )
   > Modify Transforms with Mirror_8 = RigVMFunctionReferenceNode(
    Base Modify Data="((Item=(Type=Bone,Name=\"dyn_skirt_back_l_MSTR\"),Transform=(Rotation=(X=0,Y=0,Z=0.2588,W=0.9659),Translation=(X=0,Y=0,Z=0),Scale3D=(X=1,Y=1,Z=1))),(Item=(Type=Bone,Name=\"dyn_skirt_back_mid_MSTR\"),Transform=(Rotation=(X=0,Y=0,Z=0.2588,W=0.9659),Translation=( ...(truncated)"
    Mirror Modify Data="((Item=(Type=Bone,Name=\"dyn_skirt_back_r_MSTR\"),Transform=(Rotation=(X=0,Y=0,Z=0,W=1),Translation=(X=0,Y=0,Z=0),Scale3D=(X=1,Y=1,Z=1))),(Item=(Type=Bone,Name=\"dyn_skirt_back_mid_MSTR\"),Transform=(Rotation=(X=0,Y=0,Z=0,W=1),Translation=(X=0,Y=0,Z=0),Scale3D ...(truncated)"
    Weight=RigVMVariableNode() {variable=Leg ROM_L Foot Near Pelvis}.Value
    "Mirror Weight "=RigVMVariableNode() {variable=Leg ROM_R Foot Near Pelvis}.Value
   )
   > &RigUnit_ModifyTransforms(
    ItemToModify="((Item=(Type=Bone,Name=\"dyn_skirt_front_r_MSTR\"),Transform=(Rotation=(X=0,Y=0,Z=-0.5736,W=0.8192),Translation=(X=0,Y=0,Z=0),Scale3D=(X=1,Y=1,Z=1))),(Item=(Type=Bone,Name=\"dyn_skirt_side_r_MSTR\"),Transform=(Rotation=(X=-0.0665,Y=-0.1604,Z=-0.3769,W=0.9098), ...(truncated)"
    WeightMinimum=0
    Weight=RigVMVariableNode() {variable=Leg ROM_L Foot on Thigh R}.Value
   )
   > &RigUnit_ParentConstraint(
    AdvancedSettings=(InterpolationType=Average,RotationOrderForFilter=XZY)
    Child=(Type=Bone,Name="dyn_fold_back1_MSTR")
    Filter=(TranslationFilter=(bX=true,bY=true,bZ=true),RotationFilter=(bX=True,bY=True,bZ=True),ScaleFilter=(bX=True,bY=True,bZ=True))
    Parents=((Item=(Type=Bone,Name="thigh_r"),Weight=1),(Item=(Type=Bone,Name="dyn_skirt_back_r_MSTR"),Weight=1))
    Weight=1
   )
   > &RigUnit_ParentConstraint(
    AdvancedSettings=(InterpolationType=Average,RotationOrderForFilter=XZY)
    Child=(Type=Bone,Name="dyn_fold_back2_MSTR")
    Filter=(TranslationFilter=(bX=true,bY=true,bZ=true),RotationFilter=(bX=True,bY=True,bZ=True),ScaleFilter=(bX=True,bY=True,bZ=True))
    Parents=((Item=(Type=Bone,Name="thigh_r"),Weight=1))
    Weight=1
   )
   > &RigUnit_ParentConstraint(
    AdvancedSettings=(InterpolationType=Average,RotationOrderForFilter=XZY)
    Child=(Type=Bone,Name="dyn_fold_front3_MSTR")
    Filter=(TranslationFilter=(bX=true,bY=true,bZ=true),RotationFilter=(bX=True,bY=True,bZ=True),ScaleFilter=(bX=True,bY=True,bZ=True))
    Parents=((Item=(Type=Bone,Name="thigh_r"),Weight=1))
    Weight=1
   )
   > &RigUnit_ParentConstraint(
    AdvancedSettings=(InterpolationType=Average,RotationOrderForFilter=XZY)
    Child=(Type=Bone,Name="dyn_fold_front4_MSTR")
    Filter=(TranslationFilter=(bX=true,bY=true,bZ=true),RotationFilter=(bX=True,bY=True,bZ=True),ScaleFilter=(bX=True,bY=True,bZ=True))
    Parents=((Item=(Type=Bone,Name="thigh_r"),Weight=1))
    Weight=1
   )
   > &RigUnit_ParentConstraint(
    AdvancedSettings=(InterpolationType=Average,RotationOrderForFilter=XZY)
    Child=(Type=Bone,Name="dyn_fold_front5_MSTR")
    Filter=(TranslationFilter=(bX=true,bY=true,bZ=true),RotationFilter=(bX=True,bY=True,bZ=True),ScaleFilter=(bX=True,bY=True,bZ=True))
    Parents=((Item=(Type=Bone,Name="thigh_r"),Weight=1))
    Weight=1
   )
   > &RigUnit_ModifyTransforms(
    ItemToModify="((Item=(Type=Bone,Name=\"dyn_fold_front1_MSTR\"),Transform=(Rotation=(X=0,Y=0.0436,Z=0,W=0.999),Translation=(X=0,Y=0,Z=0),Scale3D=(X=1,Y=1,Z=1))),(Item=(Type=Bone,Name=\"dyn_fold_front2_MSTR\"),Transform=(Rotation=(X=0,Y=-0.0436,Z=0,W=0.999),Translation=(X=0,Y ...(truncated)"
    Weight=1
    WeightMinimum=0
   )
   > &RigUnit_ModifyTransforms(
    ItemToModify="((Item=(Type=Bone,Name=\"dyn_fold_back1_MSTR\"),Transform=(Rotation=(X=0,Y=0.0436,Z=0,W=0.999),Translation=(X=0,Y=0,Z=0),Scale3D=(X=1,Y=1,Z=1))),(Item=(Type=Bone,Name=\"dyn_fold_back2_MSTR\"),Transform=(Rotation=(X=0,Y=-0.0436,Z=0,W=0.999),Translation=(X=0,Y=0 ...(truncated)"
    Weight=1
    WeightMinimum=0
   )
   RigVMFunction_Sequence_1 = RigVMAggregateNode(ExecuteContext=@RigVMFunction_Sequence.C)
   Solve Thigh = RigVMFunctionReferenceNode(
    Bind Front Skirt Joint=(Type=Bone,Name="dyn_skirt_front_l_MSTR")
    Bind Rear Skirt Joint=(Type=Bone,Name="dyn_skirt_back_l_MSTR")
    Bind Side Skirt Joint=(Type=Bone,Name="dyn_skirt_side_l_MSTR")
    Push Vector=(X=1.1,Y=0,Z=0)
    Thigh Joint=(Type=Bone,Name="cr_thigh_no_twist_l")
    ExecuteContext=@RigVMFunction_Sequence_1.A
    Thigh Fwd Spr=RigVMVariableNode() {variable=Leg ROM_L Thigh Raised Fwd}.Value
    Thigh Fwd Side Spr=RigVMVariableNode() {variable=Leg ROM_L Thigh Raised Fwd Side Inbetween}.Value
    Thigh Side Spr=RigVMVariableNode() {variable=Leg ROM_L Thigh Raised Side}.Value
    Thigh Side Rear Spr=RigVMVariableNode() {variable=Leg ROM_L Thigh Raised Side Rear Inbetween}.Value
    Thigh Rear Spr=RigVMVariableNode() {variable=Leg ROM_L Thigh Raised Rear}.Value
   )
   > Solve Knee = RigVMFunctionReferenceNode(
    AimAxis=(X=1,Y=0,Z=0)
    Ball=(Type=Bone,Name="foot_l")
    Bind Front Knee=(Type=Bone,Name="dyn_skirt_front_l_01")
    Bind Rear Knee=(Type=Bone,Name="dyn_skirt_back_l_MSTR")
    Bind Rear Thigh=(Type=Bone,Name="dyn_skirt_back_l_MSTR")
    Knee=(Type=Bone,Name="calf_l")
    Front Knee Bend SPR=RigVMVariableNode() {variable=Leg ROM_L Thigh Fwd Knee Bend}.Value
    Rear Knee Bend SPR=RigVMVariableNode() {variable=Leg ROM_L Thigh Rear Knee Bend}.Value
    Foot Close to Pelvis SPR=RigVMVariableNode() {variable=Leg ROM_L Foot Near Pelvis}.Value
   )
   &RigUnit_ParentConstraint(
    AdvancedSettings=(InterpolationType=Average,RotationOrderForFilter=XZY)
    Child=(Type=Bone,Name="dyn_skirt_back_mid_MSTR")
    Parents=((Item=(Type=Bone,Name="dyn_skirt_back_l_MSTR"),Weight=1),(Item=(Type=Bone,Name="dyn_skirt_back_r_MSTR"),Weight=1))
    Weight=1
    ExecutePin=@RigVMFunction_Sequence_1.C
   )
   > &RigUnit_ParentConstraint(
    AdvancedSettings=(InterpolationType=Average,RotationOrderForFilter=XZY)
    Child=(Type=Bone,Name="dyn_skirt_front_mid_MSTR")
    Parents=((Item=(Type=Bone,Name="dyn_skirt_front_l_MSTR"),Weight=1),(Item=(Type=Bone,Name="dyn_skirt_front_r_MSTR"),Weight=1))
    Weight=1
   )
   > &RigUnit_ParentConstraint(
    AdvancedSettings=(InterpolationType=Average,RotationOrderForFilter=XZY)
    Child=(Type=Bone,Name="dyn_skirt_back_mid_base")
    Parents=((Item=(Type=Bone,Name="dyn_skirt_back_l_base"),Weight=1),(Item=(Type=Bone,Name="dyn_skirt_back_r_base"),Weight=1))
    Weight=1
   )
   Solve Thigh_1 = RigVMFunctionReferenceNode(
    Bind Front Skirt Joint=(Type=Bone,Name="dyn_skirt_front_r_MSTR")
    Bind Rear Skirt Joint=(Type=Bone,Name="dyn_skirt_back_r_MSTR")
    Bind Side Skirt Joint=(Type=Bone,Name="dyn_skirt_side_r_MSTR")
    Push Vector=(X=-1.1,Y=0,Z=0)
    Thigh Joint=(Type=Bone,Name="cr_thigh_no_twist_r")
    Thigh Fwd Spr=RigVMVariableNode() {variable=Leg ROM_R Thigh Raised Fwd}.Value
    Thigh Fwd Side Spr=RigVMVariableNode() {variable=Leg ROM_R Thigh Raised Fwd Side Inbetween}.Value
    Thigh Side Spr=RigVMVariableNode() {variable=Leg ROM_R Thigh Raised Side}.Value
    Thigh Side Rear Spr=RigVMVariableNode() {variable=Leg ROM_R Thigh Raised Side Rear Inbetween}.Value
    Thigh Rear Spr=RigVMVariableNode() {variable=Leg ROM_R Thigh Raised Rear}.Value
    ExecuteContext=@RigVMFunction_Sequence_1.B
   )
   > Solve Knee_1 = RigVMFunctionReferenceNode(
    AimAxis=(X=-1,Y=0,Z=0)
    Ball=(Type=Bone,Name="foot_r")
    Bind Front Knee=(Type=Bone,Name="dyn_skirt_front_r_01")
    Bind Rear Knee=(Type=Bone,Name="dyn_skirt_back_r_MSTR")
    Bind Rear Thigh=(Type=Bone,Name="dyn_skirt_back_r_MSTR")
    Knee=(Type=Bone,Name="calf_r")
    Rear Knee Bend SPR=RigVMVariableNode() {variable=Leg ROM_R Thigh Rear Knee Bend}.Value
    Foot Close to Pelvis SPR=RigVMVariableNode() {variable=Leg ROM_R Foot Near Pelvis}.Value
    Front Knee Bend SPR=RigVMVariableNode() {variable=Leg ROM_R Thigh Fwd Knee Bend}.Value
   )
   &RigUnit_SetCurveValue(Curve=CR Thigh Raised Front L, Value=RigVMVariableNode() {variable=Leg ROM_L Thigh Raised Fwd}.Value, ExecutePin=@RigVMFunction_Sequence.F)
   > &RigUnit_SetCurveValue(Curve=CR Thigh Raised Front R, Value=RigVMVariableNode() {variable=Leg ROM_R Thigh Raised Fwd}.Value)
   > &RigUnit_SetCurveValue(Curve=CR Foot Near Pelvis L, Value=RigVMVariableNode() {variable=Leg ROM_L Foot Near Pelvis}.Value)
   > &RigUnit_SetCurveValue(Curve=CR Foot Near Pelvis R, Value=RigVMVariableNode() {variable=Leg ROM_R Foot Near Pelvis}.Value)
   > &RigUnit_SetCurveValue(Curve=CR Thigh L Across R, Value=RigVMVariableNode() {variable=Leg ROM_L Thigh Across R}.Value)
   > &RigUnit_SetCurveValue(Curve=CR Thigh L Across R, Value=RigVMVariableNode() {variable=Leg ROM_R Thigh Across L}.Value)
   Modify Transforms with Mirror_9 = RigVMFunctionReferenceNode(
    Base Modify Data="((Item=(Type=Bone,Name=\"dyn_skirt_side_l_MSTR\"),Transform=(Rotation=(X=0,Y=0,Z=-0.0872,W=0.9962),Translation=(X=0,Y=0,Z=0),Scale3D=(X=1,Y=1,Z=1))),(Item=(Type=Bone,Name=\"dyn_skirt_back_l_MSTR\"),Transform=(Rotation=(X=0,Y=0,Z=-0.0872,W=0.9962),Translation=( ...(truncated)"
    ExecuteContext=()
    Mirror Modify Data="((Item=(Type=Bone,Name=\"dyn_skirt_side_r_MSTR\"),Transform=(Rotation=(X=-0.0594,Y=-0.3368,Z=-0.1632,W=0.9254),Translation=(X=2,Y=10,Z=0),Scale3D=(X=1,Y=1,Z=1))),(Item=(Type=Bone,Name=\"dyn_skirt_back_r_MSTR\"),Transform=(Rotation=(X=-0.0151,Y=-0.173,Z=-0.0858 ...(truncated)"
    "Mirror Weight "=@n8.Result
    Weight=@n8.Result
   )
   > &RigUnit_ModifyTransforms(
    ItemToModify=((Item=(Type=Bone,Name="dyn_skirt_MSTR"),Transform=(Rotation=(X=0,Y=0,Z=0,W=1),Translation=(X=11,Y=0,Z=0),Scale3D=(X=1,Y=1,Z=1))))
    Weight=1
    WeightMinimum=0
   )
   RigVMFunction_ControlFlowBranch = &RigVMFunction_ControlFlowBranch(
    ExecuteContext=()
    Condition=&RigVMFunction_MathBoolOr(A=RigVMVariableNode() {variable=crouching}.Value, B=RigVMVariableNode() {variable=Airborne}.Value).Result
   )
   RigVMVariableNode(Value=1, ExecuteContext=@RigVMFunction_ControlFlowBranch.False) {variable=Leg ROM_Result Alpha}
   RigVMVariableNode(ExecuteContext=@RigVMFunction_ControlFlowBranch.True) {variable=Leg ROM_Result Alpha}
   n9 = &RigVMFunction_Sequence(ExecuteContext=())
   &RigVMFunction_ControlFlowBranch(ExecuteContext=@n9.A, Condition=RigVMVariableNode() {variable=Is Knee Sliding}.Value)
   &RigUnit_AimConstraintLocalSpaceOffset(
    Child=(Type=Bone,Name="dyn_skirt_back_l_01")
    ExecutePin=()
    Filter=(bX=True,bY=True,bZ=True)
    Parents=((Item=(Type=Control,Name="foot_l_ctrl"),Weight=1))
    Weight=1
    WorldUp=(Target=(X=0,Y=0,Z=1),Kind=Location,Space=(Type=Control,Name="foot_l_ctrl"))
   )
   &RigUnit_AimConstraintLocalSpaceOffset(
    AimAxis=(X=-1,Y=0,Z=0)
    Child=(Type=Bone,Name="dyn_skirt_back_r_01")
    ExecutePin=()
    Filter=(bX=True,bY=True,bZ=True)
    Parents=((Item=(Type=Control,Name="foot_r_ctrl"),Weight=1))
    Weight=1
    WorldUp=(Target=(X=0,Y=0,Z=1),Kind=Location,Space=(Type=Control,Name="foot_r_ctrl"))
   )
   Modify Transforms with Mirror_1 = RigVMFunctionReferenceNode(
    Base Modify Data="((Item=(Type=Bone,Name=\"dyn_skirt_back_l_base\"),Transform=(Rotation=(X=0,Y=0,Z=-0.0872,W=0.9962),Translation=(X=-5,Y=0,Z=5),Scale3D=(X=1,Y=1,Z=1))),(Item=(Type=Bone,Name=\"dyn_skirt_back_mid_MSTR\"),Transform=(Rotation=(X=0,Y=0,Z=-0.2588,W=0.9659),Translatio ...(truncated)"
    ExecuteContext=()
    Mirror Modify Data="((Item=(Type=Bone,Name=\"dyn_skirt_back_r_base\"),Transform=(Rotation=(X=0,Y=0,Z=0,W=1),Translation=(X=0,Y=0,Z=0),Scale3D=(X=1,Y=1,Z=1))),(Item=(Type=Bone,Name=\"dyn_skirt_back_mid_MSTR\"),Transform=(Rotation=(X=0,Y=0,Z=0,W=1),Translation=(X=0,Y=0,Z=0),Scale3D ...(truncated)"
    Weight=RigVMVariableNode() {variable=Knee_Bend_L}.Value
    "Mirror Weight "=RigVMVariableNode() {variable=Knee_Bend_r}.Value
   )
   Modify Transforms with Mirror_10 = RigVMFunctionReferenceNode(Base Modify Data=(), ExecuteContext=(), Mirror Modify Data=())
   n10 = RigVMVariableNode() {variable=Leg ROM_L Thigh Raised Fwd}
   &RigVMFunction_MathDoubleRemap(SourceMaximum=1, SourceMinimum=0, TargetMaximum=1, TargetMinimum=0.01, Value=@n10.Value)
   &RigVMFunction_MathDoubleRemap(
    SourceMaximum=0.3
    SourceMinimum=0
    TargetMaximum=1
    TargetMinimum=0.01
    Value=RigVMVariableNode() {variable=Leg ROM_L Thigh Rear Knee Bend}.Value
   )
   &RigVMFunction_MathDoubleRemap(
    SourceMaximum=0.3
    SourceMinimum=0
    TargetMaximum=1
    TargetMinimum=0.01
    Value=RigVMVariableNode() {variable=Leg ROM_R Thigh Rear Knee Bend}.Value
   )
   n11 = RigVMVariableNode() {variable=Leg ROM_R Thigh Raised Fwd}
   &RigVMFunction_MathDoubleRemap(SourceMaximum=1, SourceMinimum=0, TargetMaximum=1, TargetMinimum=0.01, Value=@n11.Value)
   RigVMFunction_MathBoolOr = &RigVMFunction_MathBoolOr(A=@n6.Value, B=RigVMVariableNode() {variable=Is Baseball Sliding}.Value)
   &RigVMFunction_MathDoubleSub(A=1, B=@n10.Value)
   &RigVMFunction_MathDoubleSub(A=1, B=@n11.Value)
   &RigVMFunction_MathDoubleSub(A=1, B=@n1.Result)
   RigVMVariableNode() {variable=Leg_Across_Back_L}
   RigVMVariableNode() {variable=Leg_Across_Back_R}
  ]}
 New Function_ContainedGraph {nodes [
   &RigUnit_AimConstraintLocalSpaceOffset(
    Child=(Type=Bone,Name="thigh_alt_l")
    ExecutePin=()
    Filter=(bX=true,bY=false,bZ=false)
    Parents=((Item=(Type=Bone,Name="thigh_r"),Weight=1))
    Weight=1
    WorldUp=(Target=(X=0,Y=0,Z=1),Kind=Direction,Space=(Type=Bone,Name="None"))
   )
   > &RigUnit_AimConstraintLocalSpaceOffset(
    Child=(Type=Bone,Name="thigh_alt_r")
    Filter=(bX=true,bY=false,bZ=false)
    Parents=((Item=(Type=Bone,Name="thigh_l"),Weight=1))
    Weight=1
    WorldUp=(Target=(X=0,Y=0,Z=1),Kind=Direction,Space=(Type=Bone,Name="None"))
   )
   RigVMFunctionEntryNode()
   RigUnit_GetTransform = &RigUnit_GetTransform(Item=(Type=Bone,Name="calf_alt_l"), bInitial=false)
   > RigVMFunction_DebugTransformMutableNoSpace = /Script/RigVM.RigVMFunction_DebugTransformMutableNoSpace(
    Color=(R=1,G=0,B=0,A=1)
    DebugDrawSettings=(DepthPriority=SDPG_Foreground,Lifetime=-1)
    Mode=Axes
    Scale=3
    Thickness=0.5
    WorldOffset=(Rotation=(X=0,Y=0,Z=0,W=1),Translation=(X=0,Y=0,Z=0),Scale3D=(X=1,Y=1,Z=1))
    bEnabled=True
    Transform=@RigUnit_GetTransform.Transform
   ) {is=struct RigVMUnitNode}
   > &RigUnit_AimConstraintLocalSpaceOffset(
    AdvancedSettings=(DebugSettings=(bEnabled=true,Scale=30,WorldOffset=(Rotation=(X=0,Y=0,Z=0,W=1),Translation=(X=0,Y=0,Z=0),Scale3D=(X=1,Y=1,Z=1))),RotationOrderForFilter=XZY)
    Child=(Type=Bone,Name="calf_alt_l")
    Filter=(bX=true,bY=false,bZ=false)
    Parents=((Item=(Type=Bone,Name="calf_r"),Weight=1))
    UpAxis=(X=0,Y=1,Z=0)
    Weight=1
    WorldUp=(Target=(X=0,Y=0,Z=1),Kind=Direction,Space=(Type=Bone,Name="None"))
   )
   > &RigUnit_AimConstraintLocalSpaceOffset(
    AdvancedSettings=(DebugSettings=(bEnabled=true,Scale=10,WorldOffset=(Rotation=(X=0,Y=0,Z=0,W=1),Translation=(X=0,Y=0,Z=0),Scale3D=(X=1,Y=1,Z=1))),RotationOrderForFilter=XZY)
    Child=(Type=Bone,Name="calf_alt_r")
    Filter=(bX=true,bY=false,bZ=false)
    Parents=((Item=(Type=Bone,Name="calf_l"),Weight=1))
    UpAxis=(X=0,Y=1,Z=0)
    Weight=1
    WorldUp=(Target=(X=0,Y=0,Z=1),Kind=Direction,Space=(Type=Bone,Name="None"))
   )
   > &RigVMFunction_DebugTransformMutableNoSpace(
    Color=(R=1,G=0,B=0,A=1)
    DebugDrawSettings=(DepthPriority=SDPG_Foreground,Lifetime=-1)
    Mode=Axes
    Scale=20
    Thickness=0.5
    WorldOffset=(Rotation=(X=0,Y=0,Z=0,W=1),Translation=(X=0,Y=0,Z=0),Scale3D=(X=1,Y=1,Z=1))
    bEnabled=true
    Transform=&RigUnit_GetTransform(Item=(Type=Bone,Name="calf_alt_l"), bInitial=False).Transform
   )
   > RigVMFunctionReturnNode()
   &RigUnit_OffsetTransformForItem(
    ExecutePin=()
    Item=(Type=Bone,Name="thigh_r")
    OffsetTransform=(Rotation=(X=0,Y=0,Z=-0.3827,W=0.9239),Translation=(X=0,Y=0,Z=0),Scale3D=(X=1,Y=1,Z=1))
    Weight=1
    bPropagateToChildren=True
   )
   &RigUnit_GetTransform(Item=(Type=Bone,Name="thigh_r"), bInitial=False)
  ]}
 RigVMFunctionLibrary {nodes [
   Robe Deformation = RigVMCollapseNode(ExecuteContext=())
   Solve Knee = RigVMCollapseNode(
    AimAxis=(X=1,Y=0,Z=0)
    Ball=(Type=None,Name="None")
    Bind Front Knee=(Type=None,Name="None")
    Bind Rear Knee=(Type=None,Name="None")
    Bind Rear Thigh=(Type=None,Name="None")
    ExecuteContext=()
    Knee=(Type=None,Name="None")
   )
   Solve Thigh = RigVMCollapseNode(
    Bind Front Skirt Joint=(Type=None,Name="None")
    Bind Rear Skirt Joint=(Type=None,Name="None")
    Bind Side Skirt Joint=(Type=None,Name="None")
    ExecuteContext=()
    Push Vector=(X=0,Y=0,Z=0)
    Thigh Joint=(Type=None,Name="None")
   )
   SolveDress = RigVMCollapseNode(ExecuteContext=())
  ]}
 RigVMFunction_Sequence_1_ContainedGraph {nodes [
   n1 = RigVMFunctionEntryNode()
   RigVMFunction_Sequence = &RigVMFunction_Sequence(ExecuteContext=@n1.ExecuteContext)
   n2 = &RigVMFunction_Sequence(ExecuteContext=@RigVMFunction_Sequence.B)
   n3 = &RigVMFunction_Sequence(ExecuteContext=@n2.B)
   n4 = &RigVMFunction_Sequence(ExecuteContext=@n3.B)
   n5 = &RigVMFunction_Sequence(ExecuteContext=@n4.B)
   n6 = &RigVMFunction_Sequence(ExecuteContext=@n5.B)
   n7 = &RigVMFunction_Sequence(ExecuteContext=@n6.B)
   RigVMFunctionReturnNode(A=@RigVMFunction_Sequence.A, B=@n2.A, C=@n3.A, D=@n4.A, E=@n5.A, F=@n6.A, G=@n7.A, H=@n7.B)
  ]}
 RigVMFunction_Sequence_1_ContainedGraph {nodes [
   n1 = RigVMFunctionEntryNode()
   RigVMFunction_Sequence = &RigVMFunction_Sequence(ExecuteContext=@n1.ExecuteContext)
   n2 = &RigVMFunction_Sequence(ExecuteContext=@RigVMFunction_Sequence.B)
   RigVMFunctionReturnNode(A=@RigVMFunction_Sequence.A, B=@n2.A, C=@n2.B)
  ]}
 RigVMFunction_Sequence_2_ContainedGraph {nodes [
   n1 = RigVMFunctionEntryNode()
   n2 = &RigVMFunction_Sequence(ExecuteContext=@n1.ExecuteContext)
   n3 = &RigVMFunction_Sequence(ExecuteContext=@n2.B)
   RigVMFunctionReturnNode(A=@n2.A, B=@n3.A, C=@n3.B)
  ]}
 RigVMModel {nodes [
   RigUnit_BeginExecution = &RigUnit_BeginExecution()
   RigVMFunction_MathQuaternionSwingTwist = &RigVMFunction_MathQuaternionSwingTwist(Input=(X=0,Y=0,Z=0,W=1), TwistAxis=(X=1,Y=0,Z=0))
   Set Transform = &RigUnit_SetRotation(
    Item=(Type=Bone,Name="calf_alt_l")
    Space=GlobalSpace
    Weight=1
    bInitial=true
    ExecutePin=@RigUnit_BeginExecution.ExecutePin
    Value=@RigVMFunction_MathQuaternionSwingTwist.Swing
   )
  ]}
]
hierarchy [
 root {loc=<0 0 0>, rot=<0 0 0 1>}
 pelvis {parent=root, loc=<0 -0.7071 93.1191>, rot=<0.0127 0.707 -0.0127 -0.707>}
 spine_01 {parent=pelvis, loc=<6.7494 -0.6388 0>, rot=<0 0 0.034 -0.9994>}
 spine_02 {parent=spine_01, loc=<7.2518 -0.035 0>, rot=<0 0 0.0238 -0.9997>}
 spine_03 {parent=spine_02, loc=<6.0104 0 0>, rot=<0 0 0.0149 -0.9999>}
 spine_04 {parent=spine_03, loc=<8.1024 0 0>, rot=<0 0 -0.0132 -0.9999>}
 spine_05 {parent=spine_04, loc=<10.0146 0 0>, rot=<0 0 -0.0015 -1>}
 neck_01 {parent=spine_05, loc=<11.3138 0.1529 0>, rot=<0 0 -0.2003 -0.9797>}
 neck_02 {parent=neck_01, loc=<4.5627 0 0>, rot=<0 0 0.0504 -0.9987>}
 head {parent=neck_02, loc=<5.3914 0 0>, rot=<0 0 0.115 -0.9934>}
 attach_cape {parent=spine_05, loc=<8.9532 -11.681 0.0524>, rot=<0.0616 -0.7044 -0.0616 -0.7044>}
 attach_backpack {parent=spine_05, loc=<-0.3552 -13.6178 0.0435>, rot=<-0.0093 -0.707 0.0092 -0.707>}
 clavicle_l {parent=spine_05, loc=<4.3575 -0.1466 -2.3223>, rot=<0.0465 -0.7289 -0.009 -0.6829>}
 upperarm_l {parent=clavicle_l, loc=<15.5111 0 0>, rot=<0.0226 -0.3234 -0.0088 -0.9459>}
 lowerarm_l {parent=upperarm_l, loc=<26.6514 0 0>, rot=<0.0391 -0.0099 -0.2453 -0.9686>}
 hand_l {parent=lowerarm_l, loc=<22.3491 0 0>, rot=<-0.7081 -0.0507 0.0975 -0.6975>}
 thumb_01_l {parent=hand_l, loc=<1.3118 -0.3191 -2.459>, rot=<0.6546 -0.4042 0.0664 -0.6354>}
 thumb_02_l {parent=thumb_01_l, loc=<5.9336 -0.0448 0.0401>, rot=<-0.0037 0.0213 0.298 -0.9543>}
 thumb_03_l {parent=thumb_02_l, loc=<2.2223 0.3682 -0.0907>, rot=<0.0118 -0.013 0.0066 -0.9998>}
 index_metacarpal_l {parent=hand_l, loc=<1.6535 0.0183 -2.96>, rot=<0.0158 -0.0124 -0.0173 -0.9996>}
 index_01_l {parent=index_metacarpal_l, loc=<8.3008 0.0357 -0.0524>, rot=<-0.0044 0.0485 0.0924 -0.9945>}
 index_02_l {parent=index_01_l, loc=<3.3587 -0.3004 0.0452>, rot=<-0.002 0.0027 0.3225 -0.9466>}
 index_03_l {parent=index_02_l, loc=<2.5578 -0.123 -0.0252>, rot=<-0.002 0.0067 0.2558 -0.9667>}
 middle_metacarpal_l {parent=hand_l, loc=<2.1856 -0.1725 -0.5984>, rot=<0.006 0.0168 -0.0329 -0.9993>}
 middle_01_l {parent=middle_metacarpal_l, loc=<7.3275 -0.2447 0.0129>, rot=<-0.0208 0.0042 0.0858 -0.9961>}
 middle_02_l {parent=middle_01_l, loc=<4.0123 0.1401 0.1175>, rot=<-0.0218 0.0258 0.3512 -0.9357>}
 middle_03_l {parent=middle_02_l, loc=<2.6077 0.0534 0.1373>, rot=<0.0082 0.0188 0.1708 -0.9851>}
 ring_metacarpal_l {parent=hand_l, loc=<2.2074 -0.3625 1.3638>, rot=<-0.0284 0.0722 -0.0458 -0.9959>}
 ring_01_l {parent=ring_metacarpal_l, loc=<7.0295 -0.6197 -0.1446>, rot=<-0.0489 -0.0252 0.1892 -0.9804>}
 ring_02_l {parent=ring_01_l, loc=<3.6174 0.4647 -0.1027>, rot=<0 0 0.2806 -0.9598>}
 ring_03_l {parent=ring_02_l, loc=<2.4152 0.114 0.0199>, rot=<0 0 0.1189 -0.9929>}
 pinky_metacarpal_l {parent=hand_l, loc=<1.7616 -0.9753 3.3814>, rot=<-0.0397 0.0814 -0.069 -0.9935>}
 pinky_01_l {parent=pinky_metacarpal_l, loc=<7.1709 -0.4831 -0.1094>, rot=<-0.0654 -0.0035 0.2242 -0.9723>}
 pinky_02_l {parent=pinky_01_l, loc=<2.8274 0.0893 -0.1043>, rot=<0.0037 0.0011 0.2854 -0.9584>}
 pinky_03_l {parent=pinky_02_l, loc=<2.2046 0.0307 -0.0078>, rot=<-0.0002 0.0098 0.1445 -0.9895>}
 deform_wrist_dn_base_l {parent=hand_l, loc=<0 -2.2342 0>, rot=<0 0 0 -1>}
 deform_wrist_dn_l {parent=deform_wrist_dn_base_l, loc=<0 0 0>, rot=<0 0 0 1>}
 deform_wrist_up_base_l {parent=hand_l, loc=<0 2.2682 0>, rot=<0 0 0 -1>}
 deform_wrist_up_l {parent=deform_wrist_up_base_l, loc=<0 0 0>, rot=<0 0 0 1>}
 lowerarm_twist_02_l {parent=lowerarm_l, loc=<9.6199 0 0>, rot=<0 0 0 -1>}
 lowerarm_twist_01_l {parent=lowerarm_l, loc=<17.8549 0 0>, rot=<0 0 0 -1>}
 deform_elbow_in_base_l {parent=lowerarm_l, loc=<0.2112 5.259 0.8274>, rot=<0 0 0 -1>}
 deform_elbow_in_l {parent=deform_elbow_in_base_l, loc=<0 0 0>, rot=<0 0 0 1>}
 deform_elbow_out_base_l {parent=lowerarm_l, loc=<0 -4 0>, rot=<0 0 0 -1>}
 deform_elbow_out_l {parent=deform_elbow_out_base_l, loc=<0 0 0>, rot=<0 0 0 1>}
 upperarm_twist_01_l {parent=upperarm_l, loc=<0.0041 0 0>, rot=<0 0 0 -1>}
 deform_upperarm_bk_base_l {parent=upperarm_twist_01_l, loc=<1.0813 -7.2193 -1.9453>, rot=<0 0 0 1>}
 deform_upperarm_bk_l {parent=deform_upperarm_bk_base_l, loc=<0 0 0>, rot=<0 0 0 1>}
 deform_upperarm_fr_base_l {parent=upperarm_twist_01_l, loc=<-0.3691 7.0506 -2.765>, rot=<0 0 0 1>}
 deform_upperarm_fr_l {parent=deform_upperarm_fr_base_l, loc=<0 0 0>, rot=<0 0 0 1>}
 deform_upperarm_up_base_l {parent=upperarm_twist_01_l, loc=<-3.2216 -1.1391 3.3311>, rot=<0 0 0 1>}
 deform_upperarm_up_l {parent=deform_upperarm_up_base_l, loc=<0 0 0>, rot=<0 0 0 1>}
 upperarm_twist_02_l {parent=upperarm_l, loc=<13.7941 0 0>, rot=<0 0 0 -1>}
 dyn_upperarm_sleeve_l_MSTR {parent=upperarm_l, loc=<-0.0001 0 -0.0001>, rot=<0 0 0 -1>}
 dyn_upperarm_l_sleeve_twist_01_MSTR {parent=dyn_upperarm_sleeve_l_MSTR, loc=<10.0366 0 -0.1>, rot=<0 0 0 1>}
 dyn_upperarm_l_sleeve_twist_01 {parent=dyn_upperarm_l_sleeve_twist_01_MSTR, loc=<0 0 0>, rot=<0 0 0 1>}
 dyn_upperarm_l_sleeve_twist_01_end {parent=dyn_upperarm_l_sleeve_twist_01, loc=<0 0 -6>, rot=<0 0 0 1>}
 dyn_upperarm_l_sleeve_twist_02_MSTR {parent=dyn_upperarm_sleeve_l_MSTR, loc=<15.3484 0 -0.1>, rot=<0 0 0 -1>}
 dyn_upperarm_l_sleeve_twist_02 {parent=dyn_upperarm_l_sleeve_twist_02_MSTR, loc=<0 0 0>, rot=<0 0 0 -1>}
 dyn_upperarm_l_sleeve_twist_02_end {parent=dyn_upperarm_l_sleeve_twist_02, loc=<0 0 -7>, rot=<0 0 0 1>}
 dyn_upperarm_l_sleeve_twist_03_MSTR {parent=dyn_upperarm_sleeve_l_MSTR, loc=<20.9241 0 -0.1>, rot=<0 0 0 -1>}
 dyn_upperarm_l_sleeve_twist_03 {parent=dyn_upperarm_l_sleeve_twist_03_MSTR, loc=<0 0 0>, rot=<0 0 0 -1>}
 dyn_upperarm_l_sleeve_twist_03_end {parent=dyn_upperarm_l_sleeve_twist_03, loc=<0 0 -8>, rot=<0 0 0 1>}
 clavicle_r {parent=spine_05, loc=<4.3575 -0.1466 2.3223>, rot=<0.7289 0.0465 -0.6829 0.009>}
 upperarm_r {parent=clavicle_r, loc=<-15.5111 0 0>, rot=<0.0226 -0.3234 -0.0088 -0.9459>}
 lowerarm_r {parent=upperarm_r, loc=<-26.6514 0 0>, rot=<0.0391 -0.0099 -0.2453 -0.9686>}
 hand_r {parent=lowerarm_r, loc=<-22.3491 0 0>, rot=<-0.7081 -0.0507 0.0975 -0.6975>}
 thumb_01_r {parent=hand_r, loc=<-1.3118 0.3191 2.459>, rot=<0.6546 -0.4042 0.0664 -0.6354>}
 thumb_02_r {parent=thumb_01_r, loc=<-5.9336 0.0448 -0.0401>, rot=<-0.0037 0.0213 0.298 -0.9543>}
 thumb_03_r {parent=thumb_02_r, loc=<-2.2223 -0.3682 0.0907>, rot=<0.0118 -0.013 0.0066 -0.9998>}
 index_metacarpal_r {parent=hand_r, loc=<-1.6535 -0.0183 2.96>, rot=<0.0158 -0.0124 -0.0173 -0.9996>}
 index_01_r {parent=index_metacarpal_r, loc=<-8.3008 -0.0357 0.0524>, rot=<-0.0044 0.0485 0.0924 -0.9945>}
 index_02_r {parent=index_01_r, loc=<-3.3587 0.3004 -0.0452>, rot=<-0.002 0.0027 0.3225 -0.9466>}
 index_03_r {parent=index_02_r, loc=<-2.5578 0.123 0.0252>, rot=<-0.002 0.0067 0.2558 -0.9667>}
 middle_metacarpal_r {parent=hand_r, loc=<-2.1856 0.1725 0.5984>, rot=<0.006 0.0168 -0.0329 -0.9993>}
 middle_01_r {parent=middle_metacarpal_r, loc=<-7.3275 0.2447 -0.0129>, rot=<-0.0208 0.0042 0.0858 -0.9961>}
 middle_02_r {parent=middle_01_r, loc=<-4.0123 -0.1401 -0.1175>, rot=<-0.0218 0.0258 0.3512 -0.9357>}
 middle_03_r {parent=middle_02_r, loc=<-2.6077 -0.0534 -0.1373>, rot=<0.0082 0.0188 0.1708 -0.9851>}
 ring_metacarpal_r {parent=hand_r, loc=<-2.2074 0.3625 -1.3638>, rot=<-0.0284 0.0722 -0.0458 -0.9959>}
 ring_01_r {parent=ring_metacarpal_r, loc=<-7.0295 0.6197 0.1446>, rot=<-0.0489 -0.0252 0.1892 -0.9804>}
 ring_02_r {parent=ring_01_r, loc=<-3.6174 -0.4647 0.1027>, rot=<0 0 0.2806 -0.9598>}
 ring_03_r {parent=ring_02_r, loc=<-2.4152 -0.114 -0.0199>, rot=<0 0 0.1189 -0.9929>}
 pinky_metacarpal_r {parent=hand_r, loc=<-1.7616 0.9753 -3.3814>, rot=<-0.0397 0.0814 -0.069 -0.9935>}
 pinky_01_r {parent=pinky_metacarpal_r, loc=<-7.1709 0.4831 0.1094>, rot=<-0.0654 -0.0035 0.2242 -0.9723>}
 pinky_02_r {parent=pinky_01_r, loc=<-2.8274 -0.0893 0.1043>, rot=<0.0037 0.0011 0.2854 -0.9584>}
 pinky_03_r {parent=pinky_02_r, loc=<-2.2046 -0.0307 0.0078>, rot=<-0.0002 0.0098 0.1445 -0.9895>}
 deform_wrist_dn_base_r {parent=hand_r, loc=<-0.0004 2.2345 -0.0001>, rot=<0 0 0 -1>}
 deform_wrist_dn_r {parent=deform_wrist_dn_base_r, loc=<0 0 0>, rot=<0 0 0 1>}
 deform_wrist_up_base_r {parent=hand_r, loc=<0.0003 -2.2684 0.0001>, rot=<0 0 0 -1>}
 deform_wrist_up_r {parent=deform_wrist_up_base_r, loc=<0 0 0>, rot=<0 0 0 1>}
 lowerarm_twist_02_r {parent=lowerarm_r, loc=<-9.6199 0 0>, rot=<0 0 0 -1>}
 lowerarm_twist_01_r {parent=lowerarm_r, loc=<-17.8549 0 0>, rot=<0 0 0 -1>}
 deform_elbow_in_base_r {parent=lowerarm_r, loc=<-0.2111 -5.2591 -0.8275>, rot=<0 0 0 -1>}
 deform_elbow_in_r {parent=deform_elbow_in_base_r, loc=<-0.0001 0 0.0002>, rot=<0 0 0 1>}
 deform_elbow_out_base_r {parent=lowerarm_r, loc=<0 4 -0.0001>, rot=<0 0 0 -1>}
 deform_elbow_out_r {parent=deform_elbow_out_base_r, loc=<0 0 0>, rot=<0 0 0 1>}
 upperarm_twist_01_r {parent=upperarm_r, loc=<-0.0041 0 0>, rot=<0 0 0 -1>}
 deform_upperarm_bk_base_r {parent=upperarm_twist_01_r, loc=<-1.0814 7.2192 1.9455>, rot=<0 0 0 1>}
 deform_upperarm_bk_r {parent=deform_upperarm_bk_base_r, loc=<0 0 0>, rot=<0 0 0 1>}
 deform_upperarm_fr_base_r {parent=upperarm_twist_01_r, loc=<0.3689 -7.0506 2.7652>, rot=<0 0 0 1>}
 deform_upperarm_fr_r {parent=deform_upperarm_fr_base_r, loc=<0 0 0>, rot=<0 0 0 1>}
 deform_upperarm_up_base_r {parent=upperarm_twist_01_r, loc=<3.2214 1.1391 -3.3308>, rot=<0 0 0 1>}
 deform_upperarm_up_r {parent=deform_upperarm_up_base_r, loc=<0 0 0>, rot=<0 0 0 -1>}
 upperarm_twist_02_r {parent=upperarm_r, loc=<-13.7941 0 0>, rot=<0 0 0 -1>}
 dyn_upperarm_sleeve_r_MSTR {parent=upperarm_r, loc=<0 0 0.0002>, rot=<0 0 0 -1>}
 dyn_upperarm_r_sleeve_twist_01_MSTR {parent=dyn_upperarm_sleeve_r_MSTR, loc=<-10.0368 0 0.1002>, rot=<0 0 0 -1>}
 dyn_upperarm_r_sleeve_twist_01 {parent=dyn_upperarm_r_sleeve_twist_01_MSTR, loc=<0 0 0>, rot=<0 0 0 -1>}
 dyn_upperarm_r_sleeve_twist_01_end {parent=dyn_upperarm_r_sleeve_twist_01, loc=<0.0004 0 5.9995>, rot=<0 0 0 -1>}
 dyn_upperarm_r_sleeve_twist_02_MSTR {parent=dyn_upperarm_sleeve_r_MSTR, loc=<-15.348 0 0.0995>, rot=<0 0 0 -1>}
 dyn_upperarm_r_sleeve_twist_02 {parent=dyn_upperarm_r_sleeve_twist_02_MSTR, loc=<0 0 0>, rot=<0 0 0 -1>}
 dyn_upperarm_r_sleeve_twist_02_end {parent=dyn_upperarm_r_sleeve_twist_02, loc=<-0.0003 0 7.0003>, rot=<0 0 0 -1>}
 dyn_upperarm_r_sleeve_twist_03_MSTR {parent=dyn_upperarm_sleeve_r_MSTR, loc=<-20.9243 0 0.1002>, rot=<0 0 0 -1>}
 dyn_upperarm_r_sleeve_twist_03 {parent=dyn_upperarm_r_sleeve_twist_03_MSTR, loc=<0 0 0>, rot=<0 0 0 -1>}
 dyn_upperarm_r_sleeve_twist_03_end {parent=dyn_upperarm_r_sleeve_twist_03, loc=<0.0004 0 7.9996>, rot=<0 0 0 -1>}
 deform_brooch {parent=clavicle_r, loc=<-12.811 -8.7106 -3.8223>, rot=<-0.5826 0.2566 0.7124 -0.2953>}
 Backpack_BR {parent=spine_05, type=Null, loc=<0 0 0>, rot=<0 0 0 1>}
 cape {parent=spine_05, type=Null, loc=<0 0 0>, rot=<0 0 0 1>}
 deform_pec_aim_base_l {parent=spine_04, loc=<7.8732 12.6388 -2.4123>, rot=<0.0102 -0.4602 0.3979 -0.7936>}
 deform_pec_aim_l {parent=deform_pec_aim_base_l, loc=<0 0 0>, rot=<0 0 0 1>}
 deform_pec_base_l {parent=deform_pec_aim_l, loc=<10.766 0 0>, rot=<0 0 0 1>}
 deform_pec_l {parent=deform_pec_base_l, loc=<0.055 0 0>, rot=<0 0 0 1>}
 deform_pec_aim_base_r {parent=spine_04, loc=<7.8727 12.6388 2.4123>, rot=<-0.4602 -0.0102 0.7936 0.3979>}
 deform_pec_aim_r {parent=deform_pec_aim_base_r, loc=<0 0 0>, rot=<0 0 0 1>}
 deform_pec_base_r {parent=deform_pec_aim_r, loc=<-10.7662 -0.0006 -0.0007>, rot=<0 0 0 1>}
 deform_pec_r {parent=deform_pec_base_r, loc=<0 0 0>, rot=<0 0 0 1>}
 dyn_fold_front1_MSTR {parent=spine_04, loc=<7.4931 12.5839 12.9878>, rot=<-0.1537 -0.1614 0.9711 0.0854>}
 dyn_fold_front1_01 {parent=dyn_fold_front1_MSTR, loc=<0 0 0>, rot=<0 0 0 1>}
 dyn_fold_front1_02 {parent=dyn_fold_front1_01, loc=<7.723 0 0>, rot=<0 0 0 1>}
 dyn_fold_front1_03 {parent=dyn_fold_front1_02, loc=<6.029 0 0>, rot=<0 0 0 1>}
 dyn_fold_front2_MSTR {parent=spine_04, loc=<4.7076 10.8629 17.0542>, rot=<0.0762 -0.2261 0.9711 -0.0068>}
 dyn_fold_front2_01 {parent=dyn_fold_front2_MSTR, loc=<0 0 0>, rot=<0 0 0 1>}
 dyn_fold_front2_02 {parent=dyn_fold_front2_01, loc=<6.3042 0 0>, rot=<0 0 0 1>}
 dyn_fold_front2_03 {parent=dyn_fold_front2_02, loc=<6.5675 0 0>, rot=<0 0 0 1>}
 dyn_fold_back1_MSTR {parent=spine_03, loc=<6.7617 -12.3035 17.8517>, rot=<0.0717 -0.9355 0.344 0.0373>}
 dyn_fold_back1_01 {parent=dyn_fold_back1_MSTR, loc=<0 0 0>, rot=<0 0 0 1>}
 dyn_fold_back1_02 {parent=dyn_fold_back1_01, loc=<13.458 0 0>, rot=<0 0 0 1>}
 dyn_fold_front3_MSTR {parent=spine_03, loc=<-2.4382 14.3248 10.134>, rot=<-0.0926 -0.202 0.975 -0.0085>}
 dyn_fold_front3_01 {parent=dyn_fold_front3_MSTR, loc=<0 0 0>, rot=<0 0 0 1>}
 thigh_l {parent=pelvis, loc=<-2.5693 -0.0109 -7.7501>, rot=<0.0012 0.0572 0.0155 -0.9982>}
 calf_l {parent=thigh_l, loc=<-39.0484 0 0>, rot=<0.0006 -0.0161 -0.0391 -0.9991>}
 foot_l {parent=calf_l, loc=<-42.561 0.1077 0.4292>, rot=<0.0418 -0.0455 0.0343 -0.9975>}
 ball_l {parent=foot_l, loc=<-6.1599 13.8264 0.6806>, rot=<0.0001 -0.0001 -0.7063 -0.7079>}
 calf_twist_02_l {parent=calf_l, loc=<-18.8064 -0.0967 0.0397>, rot=<0.0237 0.0095 -0.0035 -0.9997>}
 calf_twist_01_l {parent=calf_l, loc=<-35.813 -0.0967 0.0397>, rot=<0.0237 0.0095 -0.0035 -0.9997>}
 deform_knee_in_base_l {parent=calf_l, loc=<1.298 -5.2957 -0.0496>, rot=<0 0 0 -1>}
 deform_knee_in_l {parent=deform_knee_in_base_l, loc=<0 0 0>, rot=<0 0 0 1>}
 deform_knee_out_base_l {parent=calf_l, loc=<0 4 0>, rot=<0 0 0 -1>}
 deform_knee_out_l {parent=deform_knee_out_base_l, loc=<0 0 0>, rot=<0 0 0 1>}
 calf_alt_l {parent=calf_l, loc=<0 0 0>, rot=<0 0 0 1>}
 dyn_skirt_front_01_l {parent=calf_alt_l, loc=<-0.0181 12.4071 -0.27>, rot=<0.0235 0.1154 0.9902 0.0755>}
 dyn_skirt_front_02_l {parent=dyn_skirt_front_01_l, loc=<13.8996 0 0>, rot=<0 0 0 1>}
 dyn_skirt_front_03_l {parent=dyn_skirt_front_02_l, loc=<14.1324 0 0>, rot=<0 0 0 1>}
 dyn_skirt_side_01_l {parent=calf_alt_l, loc=<-1.8375 -2.4987 -11.1031>, rot=<-0.0122 -0.7157 -0.6969 -0.0444>}
 dyn_skirt_side_02_l {parent=dyn_skirt_side_01_l, loc=<14.4715 -0.3398 -0.0366>, rot=<-0.0077 -0.0008 0.0076 -0.9999>}
 dyn_skirt_side_03_l {parent=dyn_skirt_side_02_l, loc=<13.717 0 0>, rot=<0 0 0 1>}
 dyn_skirt_back_01_l {parent=calf_alt_l, loc=<-2.3581 -15.607 -0.6022>, rot=<-0.0021 -1 -0.0047 0.0047>}
 dyn_skirt_back_02_l {parent=dyn_skirt_back_01_l, loc=<14.1063 0 0>, rot=<0 0 0 1>}
 dyn_skirt_back_03_l {parent=dyn_skirt_back_02_l, loc=<13.7928 0 0>, rot=<0 0 0 1>}
 thigh_twist_01_l {parent=thigh_l, loc=<-21.6303 0 0>, rot=<0.0013 0 0 -1>}
 deform_groin_base_l {parent=thigh_l, loc=<-2.1908 8.1142 -1.7442>, rot=<0 0 0 -1>}
 deform_groin_l {parent=deform_groin_base_l, loc=<0 0 0>, rot=<0 0 0 1>}
 cr_thigh_l {parent=thigh_l, loc=<0 0 0>, rot=<0 0 0 1>}
 cr_thigh_no_twist_l {parent=cr_thigh_l, loc=<0 0 0>, rot=<0 0 0 1>}
 thigh_alt_l {parent=thigh_l, loc=<0 0 0>, rot=<0 0 0 1>}
 thigh_r {parent=pelvis, loc=<-2.5693 -0.0109 7.7501>, rot=<0.0572 -0.0012 0.9982 0.0155>}
 calf_r {parent=thigh_r, loc=<39.0484 0 0>, rot=<0.0006 -0.0161 -0.0391 -0.9991>}
 foot_r {parent=calf_r, loc=<42.561 -0.1077 -0.4292>, rot=<0.0418 -0.0455 0.0343 -0.9975>}
 ball_r {parent=foot_r, loc=<6.1599 -13.8264 -0.6806>, rot=<0.0001 -0.0001 -0.7063 -0.7079>}
 calf_twist_02_r {parent=calf_r, loc=<18.8064 0.0967 -0.0397>, rot=<0.0237 0.0095 -0.0035 -0.9997>}
 calf_twist_01_r {parent=calf_r, loc=<35.813 0.0967 -0.0397>, rot=<0.0237 0.0095 -0.0035 -0.9997>}
 deform_knee_in_base_r {parent=calf_r, loc=<-1.298 5.2957 0.0496>, rot=<0 0 0 -1>}
 deform_knee_in_r {parent=deform_knee_in_base_r, loc=<0 0 0>, rot=<0 0 0 1>}
 deform_knee_out_base_r {parent=calf_r, loc=<0 -4 0>, rot=<0 0 0 -1>}
 deform_knee_out_r {parent=deform_knee_out_base_r, loc=<0 0 0>, rot=<0 0 0 1>}
 calf_alt_r {parent=calf_r, loc=<0 0 0>, rot=<0 0 0 1>}
 dyn_skirt_front_01_r {parent=calf_alt_r, loc=<0.0181 -12.407 0.27>, rot=<0.0235 0.1154 0.9902 0.0755>}
 dyn_skirt_front_02_r {parent=dyn_skirt_front_01_r, loc=<-13.8996 0.0001 0>, rot=<0 0 0 1>}
 dyn_skirt_front_03_r {parent=dyn_skirt_front_02_r, loc=<-14.1324 0 0>, rot=<0 0 0 1>}
 dyn_skirt_side_01_r {parent=calf_alt_r, loc=<1.8375 2.4988 11.1031>, rot=<-0.0122 -0.7157 -0.6969 -0.0444>}
 dyn_skirt_side_02_r {parent=dyn_skirt_side_01_r, loc=<-14.4714 0.3398 0.0366>, rot=<-0.0077 -0.0008 0.0076 -0.9999>}
 dyn_skirt_side_03_r {parent=dyn_skirt_side_02_r, loc=<-13.7171 0 0>, rot=<0 0 0 1>}
 dyn_skirt_back_01_r {parent=calf_alt_r, loc=<2.3581 15.607 0.6022>, rot=<-0.0021 -1 -0.0047 0.0047>}
 dyn_skirt_back_02_r {parent=dyn_skirt_back_01_r, loc=<-14.1064 0 0.0001>, rot=<0 0 0 1>}
 dyn_skirt_back_03_r {parent=dyn_skirt_back_02_r, loc=<-13.7928 0 0>, rot=<0 0 0 1>}
 thigh_twist_01_r {parent=thigh_r, loc=<21.6303 0 0>, rot=<0.0013 0 0 -1>}
 deform_groin_base_r {parent=thigh_r, loc=<2.1908 -8.1142 1.7442>, rot=<0 0 0 -1>}
 deform_groin_r {parent=deform_groin_base_r, loc=<0 0 0>, rot=<0 0 0 1>}
 thigh_alt_r {parent=thigh_r, loc=<0 0 0>, rot=<0 0 0 1>}
 deform_thigh_notwist_l {parent=pelvis, loc=<-2.5785 -0.0168 -7.6736>, rot=<-0.0807 0.0001 0 -0.9967>}
 deform_thigh_notwist_r {parent=pelvis, loc=<-2.5386 -0.0122 7.6605>, rot=<0.0572 -0.0012 0.9982 0.0155>}
 cr_thigh_r {parent=pelvis, loc=<-2.5693 -0.0109 7.7501>, rot=<-0.0004 -0.0572 0 0.9984>}
 cr_thigh_no_twist_r {parent=cr_thigh_r, loc=<0 0 0>, rot=<0 0 0 1>}
 deform_glute_base_l {parent=pelvis, loc=<-8.8064 -8.8222 -8.4287>, rot=<0.0275 0.0501 -0.4472 -0.8926>}
 deform_glute_l {parent=deform_glute_base_l, loc=<0 0 0>, rot=<0 0 0 1>}
 deform_glute_base_r {parent=pelvis, loc=<-8.8062 -8.8222 8.4287>, rot=<0.0501 -0.0275 0.8926 -0.4472>}
 deform_glute_r {parent=deform_glute_base_r, loc=<0 0 0>, rot=<0 0 0 1>}
 dyn_skirt_ground_collision {parent=root, loc=<0 0 -8.2344>, rot=<0 0 0 1>}
 CR Foot Near Pelvis L {type=Curve, loc=<0 0 0>, rot=<0 0 0 1>}
 CR Foot Near Pelvis R {type=Curve, loc=<0 0 0>, rot=<0 0 0 1>}
 CR Rear Knee Bend L {type=Curve, loc=<0 0 0>, rot=<0 0 0 1>}
 CR Rear Knee Bend R {type=Curve, loc=<0 0 0>, rot=<0 0 0 1>}
 CR Thigh L Across R {type=Curve, loc=<0 0 0>, rot=<0 0 0 1>}
 CR Thigh R Across L {type=Curve, loc=<0 0 0>, rot=<0 0 0 1>}
 CR Thigh Raised Front L {type=Curve, loc=<0 0 0>, rot=<0 0 0 1>}
 CR Thigh Raised Front R {type=Curve, loc=<0 0 0>, rot=<0 0 0 1>}
 MaterialCurveA {type=Curve, loc=<0 0 0>, rot=<0 0 0 1>}
]
variables [
 Airborne {type=bool, default=False}
 Crouching {type=bool, default=False}
 Is Baseball Sliding {type=bool, default=False}
 Is Knee Sliding {type=bool, default=False}
 Leg ROM_Result Alpha {type=double, default=1}
 MovingFwd {type=bool, default=False}
 SlideAlphaa {type=double, default=0}
 isTacSprint {type=bool, default=False}
 kneeSlideAlpha {type=double, default=0}
 sprintAlpha {type=double, default=0}
]
