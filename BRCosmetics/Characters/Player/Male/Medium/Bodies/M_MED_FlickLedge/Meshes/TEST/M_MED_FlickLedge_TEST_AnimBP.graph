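# Test post-copy AnimBP for the M_MED_FlickLedge player body, adding secondary skirt motion to the pose copied from the attached parent mesh.
# A test Control Rig receives that pose plus baseball-slide, knee-slide and tactical-sprint flags.
# Six chained AnimDynamics nodes then simulate front, side and back skirt bones per side, using calf-driven spheres as leg collision.
# The Alpha variable weights every chain; the EventGraph has no logic setting it.
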AnimBlueprint /BRCosmetics/Characters/Player/Male/Medium/Bodies/M_MED_FlickLedge/Meshes/TEST/M_MED_FlickLedge_TEST_AnimBP
target_skeleton=&/M_MED_FlickLedge_TEST_Skeleton
graphs [
 AnimGraph {nodes [
   n1 = Control Rig!(
    Source=Copy Pose From Mesh() {bCopyCurves=true, bUseAttachedParent=true}
    Alpha
    Is Baseball Sliding:bool=Get bIsBaseballSliding()
    Is Knee Sliding:bool=Get bIsKneeSliding()
    isTacSprint:bool=Get bIsTacticalSprinting()
   ) -> (Pose) {controlRigAssetReference {blueprintRigClass=/BRCosmetics/Characters/Player/Male/Medium/Bodies/M_MED_FlickLedge/Meshes/M_MED_FlickLedge_CtrlRig_TEST.M_MED_FlickLedge_CtrlRig_TEST_C}}
   n2 = Get Alpha() -> (Alpha:float)
   n3 = "Anim Dynamics - Chain: dyn_skirt_front_01_l -> dyn_skirt_front_03_l"(ComponentPose=Local To Component(LocalPose=@n1), Alpha=@n2, ExternalForce="0,0,0") {
    angularDampingOverride=1
    angularSpringConstant=5
    bAngularSpring=true
    bChain=true
    bGravityOverrideInSimSpace=true
    bOverrideAngularDamping=true
    bUseSphericalLimits=true
    gravityScale=0.3
    numSolverIterationsPostUpdate=8
    numSolverIterationsPreUpdate=32
    simulationSpace=RootRelative
    alphaBoolBlend {blendOption=Linear}
    boundBone {boneName=dyn_skirt_front_01_l}
    chainEnd {boneName=dyn_skirt_front_03_l}
    componentAppliedLinearAccClamp=<1000 1000 1000>
    componentLinearAccScale=<1 1 1>
    componentLinearVelScale=<1 1 1>
    physicsBodyDefinitions [
     {
      sphereCollisionRadius=0
      boundBone {boneName=dyn_skirt_front_01_l}
      boxExtents=<18 18 30>
      constraintSetup {angularLimitsMax=<10 35 10>, angularLimitsMin=<-10 -20 -35>, angularTarget=<0 0 1>}
      localJointOffset=<5 0 0>
     }
     {
      boundBone {boneName=dyn_skirt_front_02_l}
      boxExtents=<18 18 30>
      constraintSetup {angularLimitsMax=<10 35 10>, angularLimitsMin=<-10 -20 -35>, angularTarget=<0 0 1>}
      localJointOffset=<5 0 0>
     }
     {
      boundBone {boneName=dyn_skirt_front_03_l}
      boxExtents=<18 18 30>
      constraintSetup {angularLimitsMax=<10 35 10>, angularLimitsMin=<-10 -20 -35>, angularTarget=<0 0 1>}
      localJointOffset=<5 0 0>
     }
    ]
    relativeSpaceBone {boneName=thigh_l}
    sphericalLimits [
     {limitRadius=13, drivingBone {boneName=calf_l}, sphereLocalOffset=<0 0 0>}
     {limitRadius=13, drivingBone {boneName=calf_l}, sphereLocalOffset=<-25 0 0>}
     {limitRadius=13, drivingBone {boneName=calf_l}, sphereLocalOffset=<-40 0 0>}
    ]
   }
   n4 = "Anim Dynamics - Chain: dyn_skirt_side_01_l -> dyn_skirt_side_03_l"(ComponentPose=@n3, Alpha=@n2, ExternalForce="0,0,0") {
    angularDampingOverride=1
    angularSpringConstant=5
    bAngularSpring=true
    bChain=true
    bGravityOverrideInSimSpace=true
    bOverrideAngularDamping=true
    bUseSphericalLimits=true
    gravityScale=0.3
    numSolverIterationsPostUpdate=8
    numSolverIterationsPreUpdate=32
    simulationSpace=RootRelative
    alphaBoolBlend {blendOption=Linear}
    boundBone {boneName=dyn_skirt_side_01_l}
    chainEnd {boneName=dyn_skirt_side_03_l}
    componentAppliedLinearAccClamp=<1000 1000 1000>
    componentLinearAccScale=<1 1 1>
    componentLinearVelScale=<1 1 1>
    physicsBodyDefinitions [
     {
      sphereCollisionRadius=0
      boundBone {boneName=dyn_skirt_side_01_l}
      boxExtents=<18 18 30>
      constraintSetup {angularLimitsMax=<10 35 10>, angularLimitsMin=<-10 -20 -35>, angularTarget=<0 0 1>}
      localJointOffset=<5 0 0>
     }
     {
      sphereCollisionRadius=0
      boundBone {boneName=dyn_skirt_side_02_l}
      boxExtents=<18 18 31>
      constraintSetup {angularLimitsMax=<10 35 10>, angularLimitsMin=<-10 -20 -35>, angularTarget=<0 0 1>}
      localJointOffset=<5 0 0>
     }
     {
      sphereCollisionRadius=0
      boundBone {boneName=dyn_skirt_side_03_l}
      boxExtents=<18 18 30>
      constraintSetup {angularLimitsMax=<10 35 10>, angularLimitsMin=<-10 -20 -35>, angularTarget=<0 0 1>}
      localJointOffset=<5 0 0>
     }
    ]
    relativeSpaceBone {boneName=thigh_l}
    sphericalLimits [
     {limitRadius=13, drivingBone {boneName=calf_l}, sphereLocalOffset=<0 0 0>}
     {limitRadius=13, drivingBone {boneName=calf_l}, sphereLocalOffset=<-25 0 0>}
     {limitRadius=13, drivingBone {boneName=calf_l}, sphereLocalOffset=<-40 0 0>}
    ]
   }
   n5 = "Anim Dynamics - Chain: dyn_skirt_back_01_l -> dyn_skirt_back_03_l"(ComponentPose=@n4, Alpha=@n2, ExternalForce="0,0,0") {
    angularDampingOverride=1
    angularSpringConstant=5
    bAngularSpring=true
    bChain=true
    bGravityOverrideInSimSpace=true
    bOverrideAngularDamping=true
    bUseSphericalLimits=true
    gravityScale=0.3
    numSolverIterationsPostUpdate=8
    numSolverIterationsPreUpdate=32
    simulationSpace=RootRelative
    alphaBoolBlend {blendOption=Linear}
    boundBone {boneName=dyn_skirt_back_01_l}
    chainEnd {boneName=dyn_skirt_back_03_l}
    componentAppliedLinearAccClamp=<1000 1000 1000>
    componentLinearAccScale=<1 1 1>
    componentLinearVelScale=<1 1 1>
    physicsBodyDefinitions [
     {
      boundBone {boneName=dyn_skirt_back_01_l}
      boxExtents=<18 18 30>
      constraintSetup {angularLimitsMax=<10 15 10>, angularLimitsMin=<-10 -20 -15>, angularTarget=<0 0 1>}
      localJointOffset=<5 0 0>
     }
     {
      boundBone {boneName=dyn_skirt_back_02_l}
      boxExtents=<18 18 30>
      constraintSetup {angularLimitsMax=<10 15 10>, angularLimitsMin=<-10 -20 -15>, angularTarget=<0 0 1>}
      localJointOffset=<5 0 0>
     }
     {
      boundBone {boneName=dyn_skirt_back_03_l}
      boxExtents=<18 18 30>
      constraintSetup {angularLimitsMax=<10 15 10>, angularLimitsMin=<-10 -20 -15>, angularTarget=<0 0 1>}
      localJointOffset=<5 0 0>
     }
    ]
    relativeSpaceBone {boneName=thigh_l}
    sphericalLimits [
     {limitRadius=13, drivingBone {boneName=calf_l}, sphereLocalOffset=<0 0 0>}
     {limitRadius=13, drivingBone {boneName=calf_l}, sphereLocalOffset=<-25 0 0>}
     {limitRadius=13, drivingBone {boneName=calf_l}, sphereLocalOffset=<-40 0 0>}
    ]
   }
   n6 = "Anim Dynamics - Chain: dyn_skirt_front_01_r -> dyn_skirt_front_03_r"(ComponentPose=@n5, Alpha=@n2, ExternalForce="0,0,0") {
    angularDampingOverride=1
    angularSpringConstant=5
    bAngularSpring=true
    bChain=true
    bGravityOverrideInSimSpace=true
    bOverrideAngularDamping=true
    bUseSphericalLimits=true
    gravityScale=0.3
    numSolverIterationsPostUpdate=8
    numSolverIterationsPreUpdate=32
    simulationSpace=RootRelative
    alphaBoolBlend {blendOption=Linear}
    boundBone {boneName=dyn_skirt_front_01_r}
    chainEnd {boneName=dyn_skirt_front_03_r}
    componentAppliedLinearAccClamp=<1000 1000 1000>
    componentLinearAccScale=<1 1 1>
    componentLinearVelScale=<1 1 1>
    physicsBodyDefinitions [
     {
      boundBone {boneName=dyn_skirt_front_01_r}
      boxExtents=<18 18 30>
      constraintSetup {angularLimitsMax=<10 15 10>, angularLimitsMin=<-10 -20 -15>, angularTarget=<0 0 1>}
      localJointOffset=<-5 0 0>
     }
     {
      boundBone {boneName=dyn_skirt_front_02_r}
      boxExtents=<18 18 30>
      constraintSetup {angularLimitsMax=<10 15 10>, angularLimitsMin=<-10 -20 -15>, angularTarget=<0 0 1>}
      localJointOffset=<-5 0 0>
     }
     {
      boundBone {boneName=dyn_skirt_front_03_r}
      boxExtents=<18 18 30>
      constraintSetup {angularLimitsMax=<10 15 10>, angularLimitsMin=<-10 -20 -15>, angularTarget=<0 0 1>}
      localJointOffset=<-5 0 0>
     }
    ]
    relativeSpaceBone {boneName=thigh_l}
    sphericalLimits [
     {limitRadius=13, drivingBone {boneName=calf_r}, sphereLocalOffset=<0 0 0>}
     {limitRadius=13, drivingBone {boneName=calf_r}, sphereLocalOffset=<25 0 0>}
     {limitRadius=13, drivingBone {boneName=calf_r}, sphereLocalOffset=<40 0 0>}
    ]
   }
   n7 = "Anim Dynamics - Chain: dyn_skirt_side_01_r -> dyn_skirt_side_03_r"(ComponentPose=@n6, Alpha=@n2, ExternalForce="0,0,0") {
    angularDampingOverride=1
    angularSpringConstant=5
    bAngularSpring=true
    bChain=true
    bGravityOverrideInSimSpace=true
    bOverrideAngularDamping=true
    bUseSphericalLimits=true
    gravityScale=0.3
    numSolverIterationsPostUpdate=8
    numSolverIterationsPreUpdate=32
    simulationSpace=RootRelative
    alphaBoolBlend {blendOption=Linear}
    boundBone {boneName=dyn_skirt_side_01_r}
    chainEnd {boneName=dyn_skirt_side_03_r}
    componentAppliedLinearAccClamp=<1000 1000 1000>
    componentLinearAccScale=<1 1 1>
    componentLinearVelScale=<1 1 1>
    physicsBodyDefinitions [
     {
      boundBone {boneName=dyn_skirt_side_01_r}
      boxExtents=<18 18 30>
      constraintSetup {angularLimitsMax=<10 15 10>, angularLimitsMin=<-10 -20 -15>, angularTarget=<0 0 1>}
      localJointOffset=<-5 0 0>
     }
     {
      boundBone {boneName=dyn_skirt_side_02_r}
      boxExtents=<18 18 30>
      constraintSetup {angularLimitsMax=<10 15 10>, angularLimitsMin=<-10 -20 -15>, angularTarget=<0 0 1>}
      localJointOffset=<-5 0 0>
     }
     {
      boundBone {boneName=dyn_skirt_side_03_r}
      boxExtents=<18 18 30>
      constraintSetup {angularLimitsMax=<10 15 10>, angularLimitsMin=<-10 -20 -15>, angularTarget=<0 0 1>}
      localJointOffset=<-5 0 0>
     }
    ]
    relativeSpaceBone {boneName=thigh_l}
    sphericalLimits [
     {limitRadius=13, drivingBone {boneName=calf_r}, sphereLocalOffset=<0 0 0>}
     {limitRadius=13, drivingBone {boneName=calf_r}, sphereLocalOffset=<25 0 0>}
     {limitRadius=13, drivingBone {boneName=calf_r}, sphereLocalOffset=<40 0 0>}
    ]
   }
   n8 = "Anim Dynamics - Chain: dyn_skirt_back_01_r -> dyn_skirt_back_03_r"(ComponentPose=@n7, Alpha=@n2, ExternalForce="0,0,0") {
    angularDampingOverride=1
    angularSpringConstant=5
    bAngularSpring=true
    bChain=true
    bGravityOverrideInSimSpace=true
    bOverrideAngularDamping=true
    bUseSphericalLimits=true
    gravityScale=0.3
    numSolverIterationsPostUpdate=8
    numSolverIterationsPreUpdate=32
    simulationSpace=RootRelative
    alphaBoolBlend {blendOption=Linear}
    boundBone {boneName=dyn_skirt_back_01_r}
    chainEnd {boneName=dyn_skirt_back_03_r}
    componentAppliedLinearAccClamp=<1000 1000 1000>
    componentLinearAccScale=<1 1 1>
    componentLinearVelScale=<1 1 1>
    physicsBodyDefinitions [
     {
      boundBone {boneName=dyn_skirt_back_01_r}
      boxExtents=<18 18 30>
      constraintSetup {angularLimitsMax=<10 15 10>, angularLimitsMin=<-10 -20 -15>, angularTarget=<0 0 1>}
      localJointOffset=<-5 0 0>
     }
     {
      boundBone {boneName=dyn_skirt_back_02_r}
      boxExtents=<18 18 30>
      constraintSetup {angularLimitsMax=<10 15 10>, angularLimitsMin=<-10 -20 -15>, angularTarget=<0 0 1>}
      localJointOffset=<-5 0 0>
     }
     {
      boundBone {boneName=dyn_skirt_back_03_r}
      boxExtents=<18 18 30>
      constraintSetup {angularLimitsMax=<10 15 10>, angularLimitsMin=<-10 -20 -15>, angularTarget=<0 0 1>}
      localJointOffset=<-5 0 0>
     }
    ]
    relativeSpaceBone {boneName=thigh_l}
    sphericalLimits [
     {limitRadius=13, drivingBone {boneName=calf_r}, sphereLocalOffset=<0 0 0>}
     {limitRadius=13, drivingBone {boneName=calf_r}, sphereLocalOffset=<25 0 0>}
     {limitRadius=13, drivingBone {boneName=calf_r}, sphereLocalOffset=<40 0 0>}
    ]
   }
   Output Pose(Result=Component To Local(ComponentPose=@n8))
  ]}
 EventGraph {nodes [
   TryGetPawnOwner()
   Event BlueprintUpdateAnimation()
  ]}
]
variables [Alpha {type=number}]
